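# Creative property editor row widget that shows a device button component as a named text button.
# Post-init sets the name, subscribes to label changes and fills the button label.
# A click broadcasts the component through FortPlayerControllerAthena; selection restyles background, name and modified icon, shown only for changed values.
# The label update path also binds widget activation to a leftover "Hello" debug print.
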WidgetBlueprint /Game/Athena/UI/InGame/Creative/PropertyEditor/EditWidgets/CreativePropertyEditWidgetBase_DeviceButtonComponent_Basic
graphs [
 BP_GetDesiredFocusTarget {nodes [
   Get Desired Focus Target()
   > Return Node!(execute, ReturnValue:Widget Object Reference=Get CreativeTextButton() -> (CreativeTextButton:Icon Text Button Object Reference))
  ]}
 EventGraph {nodes [
   "On Button Clicked (CreativeTextButton)"!() -> (OutputDelegate, then, Button:Common Button Legacy Object Reference) {is=title K2Node_ComponentBoundEvent}
   > n1 = Cast To FortPlayerControllerAthena(Object=GetOwningPlayer())
   > BroadcastDeviceButtonComponentClickedDelegate!(
    execute
    self:Fort Player Controller Athena Object Reference=@n1
    DeviceButtonComponent:Fort Creative Device Button Component Object Reference=GetDeviceButtonComponent!(self:Creative Property Edit Widget Device Button Component Base Object Reference) -> (ReturnValue:Fort Creative Device Button Component Object Reference)
   ) -> (then)
   n2 = OnDeviceButtonLabelChanged() {is=title K2Node_CustomEvent}
   n3 = Set Text!(
    execute=[@n8.then, @n2.then]
    self:Icon Text Button Object Reference=Get CreativeTextButton()
    ButtonText:Text=GetButtonLabel!(self:Creative Property Edit Widget Device Button Component Base Object Reference) -> (ReturnValue:Text)
   ) -> (then) {is=title K2Node_CallFunction}
   n4 = CustomEvent() {is=title K2Node_CustomEvent}
   Bind Event to On Widget Activated(execute=@n3.then, self:Common Activatable Widget Object Reference, Delegate=@n4) {is=title K2Node_AddDelegate}
   n5 = Event BP_OnModified!() -> (OutputDelegate, then, bModified:bool) {is=title K2Node_Event}
   > UpdateModifiedState!(execute, self, isModifiedFromDefaultValue:bool=@n5.bModified) -> (then)
   n6 = Event OnEditWidgetSelectedChanged!() -> (OutputDelegate, then, bSelected:bool) {is=title K2Node_Event}
   > Update Selected State!(execute, self, Selected:bool=@n6.bSelected) -> (then)
   n7 = Event SetDesiredValueWidth!() -> (OutputDelegate, then, DesiredValueWidth:float) {is=title K2Node_Event}
   > SetWidthOverride(self=Get SizeBox_Value() -> (SizeBox_Value:Size Box Object Reference), InWidthOverride=@n7.DesiredValueWidth)
   Event BP_PostInitEditWidget() {is=title K2Node_Event}
   > "SetText (Text)"(
    self=Get TextBlock_OptionDisplayName() -> (TextBlock_OptionDisplayName:Common Text Object Reference)
    InText=GetEntryDisplayName!(self:Creative Property Edit Widget Base Object Reference) -> (ReturnValue:Text)
   )
   > n8 = Assign On Device Button Label Changed Delegate!(execute:Exec, self:Fort Creative Device Button Component Object Reference=GetDeviceButtonComponent(), Delegate:"Delegate (by ref)"=@n2) -> (then:Exec) {is=title K2Node_AssignDelegate}
   PrintString!(execute=@n4.then, InString=Hello, bPrintToScreen, bPrintToLog, TextColor, Duration, Key) -> (then)
  ]}
 OnFocusReceived {nodes [
   OnFocusReceived!() -> (then, MyGeometry:Geometry Structure, InFocusEvent:Focus Event Structure) {is=title K2Node_FunctionEntry}
   > Return Node(ReturnValue:Event Reply Structure=Handled())
  ]}
 Update Selected State {nodes [
   n1 = Update Selected State!() -> (then, Selected:bool=~) {is=title K2Node_FunctionEntry}
   > n2 = Branch(Condition=@n1)
   n3 = Get WBP_Background_SkewedWithMarker() -> (WBP_Background_SkewedWithMarker:WBP Creative Property Editor Background Skewed With Marker Object Reference)
   > n4 = Event Update Background Style!(
    execute
    self:WBP Creative Property Editor Background Skewed With Marker Object Reference=@n3
    InType:E_UI_CreativePropertyEditor_Type Enum=NewEnumerator3
    InMarkerState:E_UI_CreativePropertyEditor_MarkerStates Enum=NewEnumerator0
    InFocusState:E_UI_CreativePropertyEditor_FocusStates Enum=NewEnumerator1
   ) -> (then)
   n5 = Select(
    Option 0:Linear Color Structure=Get Non Selected Color() -> (Non Selected Color:Linear Color Structure)
    Option 1:Linear Color Structure=Get Selected Color() -> (Selected Color:Linear Color Structure)
    Index=Get Selected() -> (Selected:bool)
   ) -> (ReturnValue:Linear Color Structure)
   SetColorAndOpacity(
    execute=[@n6.then, @n4.then]
    self:Image Object Reference=Get Image_Modified() -> (Image_Modified:Image Object Reference)
    InColorAndOpacity:Linear Color Structure=@n5
   )
   > SetColorAndOpacity(self:Text Object Reference=Get TextBlock_OptionDisplayName(), InColorAndOpacity:Slate Color Structure=Make Slate Color(SpecifiedColor=@n5))
   n6 = Event Update Background Style(execute=@n2.else, self=@n3, InType=NewEnumerator4, InMarkerState=NewEnumerator0, InFocusState=NewEnumerator0)
  ]}
 UpdateModifiedState {nodes [
   n1 = UpdateModifiedState!() -> (then, isModifiedFromDefaultValue:bool=~) {is=title K2Node_FunctionEntry}
   > SetVisibility(
    self=Get Image_Modified()
    InVisibility=Select(Option 0:ESlateVisibility Enum=Hidden, Option 1:ESlateVisibility Enum=SelfHitTestInvisible, Index=@n1) -> (ReturnValue:ESlateVisibility Enum)
   )
  ]}
]
variables [
 Non Selected Color {type=LinearColor, default=<rgba 0.3372 0.402 0.5906 1>}
 Selected Color {type=LinearColor, default=<rgba 0.1912 0.2307 0.2874 1>}
]
widgets {
 info {
  inheritedWidgetCount=1
  parentClass=/Script/FortniteUI.CreativePropertyEditWidgetDeviceButtonComponentBase
  rootWidgetClass=&Overlay
  widgetCount=14
 }
 widgets [
  Overlay_25 {widgetClassPath=&Overlay, widget=Overlay_25}
  SizeBox_2 {
   widgetClassPath=&SizeBox
   parent=Overlay_25
   widget=SizeBox_2
   properties {bOverride_HeightOverride=true, heightOverride=82}
   slot {class=OverlaySlot, properties {horizontalAlignment=HAlign_Fill, verticalAlignment=VAlign_Center, padding=<ltrb 16 0 16 0>}}
  }
  Overlay_0 {widgetClassPath=&Overlay, parent=SizeBox_2, widget=Overlay_0, slot {class=SizeBoxSlot}}
  WBP_Background_SkewedWithMarker {
   widgetClassPath=/Game/Athena/UI/InGame/Creative/PropertyEditor/WBP_CreativePropertyEditor_Background_SkewedWithMarker
   bIsVariable=true
   parent=Overlay_0
   widget=WBP_Background_SkewedWithMarker
   properties {type=Default}
   slot {class=OverlaySlot, properties {horizontalAlignment=HAlign_Fill, verticalAlignment=VAlign_Fill}}
  }
  HorizontalBox_2 {
   widgetClassPath=&HorizontalBox
   parent=Overlay_0
   widget=HorizontalBox_2
   slot {class=OverlaySlot, properties {horizontalAlignment=HAlign_Fill, verticalAlignment=VAlign_Fill}}
  }
  Image_Modified {
   widgetClassPath=&Image
   bIsVariable=true
   parent=HorizontalBox_2
   widget=Image_Modified
   properties {
    visibility=Hidden
    brush {
     bIsDynamicallyLoaded=false
     drawAs=Image
     imageType=NoImage
     mirroring=NoMirror
     resourceName=None
     resourceObject=/Game/Athena/UI/InGame/Creative/PropertyEditor/Art/Icon-PropertyEditor-Modified
     tiling=NoTile
     imageSize=<20 20>
     margin=<ltrb 0 0 0 0>
     outlineSettings {
      bUseBrushTransparency=false
      roundingType=HalfHeightRadius
      width=0
      color {colorUseRule=UseColor_Specified, specifiedColor=<rgba 0 0 0 0>}
      cornerRadii=<0 0 0 0>
     }
     tintColor {colorUseRule=UseColor_Specified, specifiedColor=<rgba 1 1 1 1>}
     uVRegion {bIsValid=false, max=<0 0>, min=<0 0>}
    }
    colorAndOpacity=<rgba 0.3372 0.402 0.5906 1>
   }
   slot {class=HorizontalBoxSlot, properties {horizontalAlignment=HAlign_Left, verticalAlignment=VAlign_Center, padding=<ltrb 6 0 4 0>}}
  }
  SizeBox_0 {
   widgetClassPath=&SizeBox
   parent=HorizontalBox_2
   widget=SizeBox_0
   properties {bOverride_WidthOverride=true, widthOverride=500}
   slot {class=HorizontalBoxSlot, properties {verticalAlignment=VAlign_Center, padding=<ltrb 0 0 24 0>, size {sizeRule=Fill, value=1}}}
  }
  TextBlock_OptionDisplayName {
   widgetClassPath=&CommonTextBlock
   bIsVariable=true
   parent=SizeBox_0
   widget=TextBlock_OptionDisplayName
   properties {
    autoWrapText=true
    scrollStyle=&TextScrollStyle-Base_C
    text="[Parameter name]"
    colorAndOpacity {colorUseRule=UseColor_Specified, specifiedColor=<rgba 0.3372 0.402 0.5906 1>}
    font {
     bForceMonospaced=false
     bMaterialIsStencil=false
     fontMaterial=null
     fontObject=&BurbankSmall
     letterSpacing=0
     monospacedWidth=1
     size=20
     skewAmount=0
     typefaceFontName=Bold
     outlineSettings {
      bApplyOutlineToDropShadows=false
      bMiteredCorners=false
      bSeparateFillAlpha=false
      outlineMaterial=null
      outlineSize=0
      outlineColor=<rgba 0.0052 0.0052 0.0056 0.2>
     }
    }
   }
   slot {class=SizeBoxSlot}
  }
  SizeBox_Value {
   widgetClassPath=&SizeBox
   bIsVariable=true
   parent=HorizontalBox_2
   widget=SizeBox_Value
   properties {bOverride_WidthOverride=true, widthOverride=320}
   slot {class=HorizontalBoxSlot, properties {verticalAlignment=VAlign_Center}}
  }
  Overlay_2 {widgetClassPath=&Overlay, parent=SizeBox_Value, widget=Overlay_2, slot {class=SizeBoxSlot}}
  OptionSelectionArea_2 {
   widgetClassPath=&HorizontalBox
   bIsVariable=true
   parent=Overlay_2
   widget=OptionSelectionArea_2
   slot {class=OverlaySlot, properties {horizontalAlignment=HAlign_Fill}}
  }
  CreativeTextButton {
   widgetClassPath=&IconTextButton
   bIsVariable=true
   parent=OptionSelectionArea_2
   widget=CreativeTextButton
   properties {
    onButtonBaseClicked=(&self:WidgetTree.CreativeTextButton.HandleOnButtonClicked)
    onButtonBaseDoubleClicked=(&self:WidgetTree.CreativeTextButton.HandleOnButtonDoubleClicked)
    onButtonBaseHovered=(&self:WidgetTree.CreativeTextButton.HandleOnButtonHovered)
    onButtonBaseUnhovered=(&self:WidgetTree.CreativeTextButton.HandleOnButtonUnhovered)
    onSelectedChangedBase=(&self:WidgetTree.CreativeTextButton.HandleOnSelectedChanged)
    style=/FortUILibrary/Foundation/Buttons/ButtonStyles/Athena/Athena_ButtonStyle_AngledBlueMenuCreativeButton.Athena_ButtonStyle_AngledBlueMenuCreativeButton_C
    triggeringInputAction {dataTable=/FortUILibrary/Input/CreativeInputActionDataTable, rowName=Generic_Accept}
   }
   slot {class=HorizontalBoxSlot, properties {size {sizeRule=Fill, value=1}}}
  }
  Button_Collapse {
   widgetClassPath=&IconTextButton
   bInherited=true
   parent=Overlay_25
   widget=Button_Collapse
   properties {
    onButtonBaseClicked=(&self:WidgetTree.Button_Collapse.HandleOnButtonClicked)
    onButtonBaseDoubleClicked=(&self:WidgetTree.Button_Collapse.HandleOnButtonDoubleClicked)
    onButtonBaseHovered=(&self:WidgetTree.Button_Collapse.HandleOnButtonHovered)
    onButtonBaseUnhovered=(&self:WidgetTree.Button_Collapse.HandleOnButtonUnhovered)
    onSelectedChangedBase=(&self:WidgetTree.Button_Collapse.HandleOnSelectedChanged)
    visibility=Collapsed
   }
   slot {class=OverlaySlot, properties {horizontalAlignment=HAlign_Center, verticalAlignment=VAlign_Center}}
  }
  SaveSpinner_2 {
   widgetClassPath=&Image
   bIsVariable=true
   parent=Overlay_25
   widget=SaveSpinner_2
   properties {
    visibility=Collapsed
    brush {
     bIsDynamicallyLoaded=false
     drawAs=Image
     imageType=NoImage
     mirroring=NoMirror
     resourceName=None
     resourceObject=/FortUILibrary/UIKit/Throbber/MI_UIKit_Color_Throbber
     tiling=NoTile
     imageSize=<24 24>
     margin=<ltrb 0 0 0 0>
     outlineSettings {
      bUseBrushTransparency=false
      roundingType=HalfHeightRadius
      width=0
      color {colorUseRule=UseColor_Specified, specifiedColor=<rgba 0 0 0 0>}
      cornerRadii=<0 0 0 0>
     }
     tintColor {colorUseRule=UseColor_Specified, specifiedColor=<rgba 1 1 1 1>}
     uVRegion {bIsValid=false, max=<0 0>, min=<0 0>}
    }
   }
   slot {class=OverlaySlot, properties {verticalAlignment=VAlign_Center, padding=<ltrb 18 -24 0 0>}}
  }
 ]
}
