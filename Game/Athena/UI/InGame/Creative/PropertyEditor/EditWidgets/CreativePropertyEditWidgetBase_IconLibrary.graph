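# Creative property editor row widget that lets the player choose an icon option, showing the selected swatch and uppercased name.
# Post-init casts DialogWidget to an icon picker, fills it from IconOptionObjects and binds its events.
# Click, touch or Accept opens the picker at the current index, while committing a swatch stores the index, marks modified, closes and refreshes the display.
# ActionConfirmed and ActionCancelled animations swap the arrows and Accept/Back buttons.
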WidgetBlueprint /Game/Athena/UI/InGame/Creative/PropertyEditor/EditWidgets/CreativePropertyEditWidgetBase_IconLibrary
animations [
 ActionCancelled {
  binding_count=4
  length_seconds=0.25
  object=&self:ActionCancelled
  sequence {bindings [
    Image_CloseArrow {
     id=#1
     possessed_class=Image
     tracks [
      {
       type=MovieSceneFloatTrack
       display_name=Render Opacity
       sections [{type=MovieSceneFloatSection, end_frame=5, start_frame=0, channels [None {type=MovieSceneScriptingFloatChannel, keys [[0,1],[5,0]]}]}]
      }
      {
       type=MovieScene2DTransformTrack
       sections [{type=MovieScene2DTransformSection, end_frame=5, start_frame=0, channels [Angle {type=MovieSceneScriptingFloatChannel, keys [[0,0],[5,-90]]}]}]
      }
     ]
    }
    UserActionMenuInputButton_Accept {
     id=#2
     possessed_class=UserActionMenuInputButton_C
     tracks [{
       type=MovieSceneFloatTrack
       display_name=Render Opacity
       sections [{type=MovieSceneFloatSection, end_frame=5, start_frame=0, channels [None {type=MovieSceneScriptingFloatChannel, keys [[0,0],[5,1]]}]}]
      }]
    }
    Image_OpenArrow {
     id=#3
     possessed_class=Image
     tracks [
      {
       type=MovieSceneFloatTrack
       display_name=Render Opacity
       sections [{type=MovieSceneFloatSection, end_frame=5, start_frame=0, channels [None {type=MovieSceneScriptingFloatChannel, keys [[0,0],[5,1]]}]}]
      }
      {
       type=MovieScene2DTransformTrack
       sections [{
         type=MovieScene2DTransformSection
         end_frame=5
         start_frame=0
         channels [
          Translation.X {type=MovieSceneScriptingFloatChannel, keys [[0,10],[5,0]]}
          Scale.X {type=MovieSceneScriptingFloatChannel, keys [[0,1.2],[5,1]]}
          Scale.Y {type=MovieSceneScriptingFloatChannel, keys [[0,1.2],[5,1]]}
         ]
        }]
      }
     ]
    }
    UserActionMenuInputButton_Back {
     id=#4
     possessed_class=UserActionMenuInputButton_C
     tracks [{
       type=MovieSceneFloatTrack
       display_name=Render Opacity
       sections [{type=MovieSceneFloatSection, end_frame=5, start_frame=0, channels [None {type=MovieSceneScriptingFloatChannel, keys [[0,1],[5,0]]}]}]
      }]
    }
   ]}
 }
 ActionConfirmed {
  binding_count=4
  length_seconds=0.25
  object=&self:ActionConfirmed
  sequence {bindings [
    Image_CloseArrow {
     id=#5
     possessed_class=Image
     tracks [
      {
       type=MovieSceneFloatTrack
       display_name=Render Opacity
       sections [{type=MovieSceneFloatSection, end_frame=5, start_frame=0, channels [None {type=MovieSceneScriptingFloatChannel, keys [[0,0],[5,1]]}]}]
      }
      {
       type=MovieScene2DTransformTrack
       sections [{type=MovieScene2DTransformSection, end_frame=5, start_frame=0, channels [Angle {type=MovieSceneScriptingFloatChannel, keys [[0,-90],[5,0]]}]}]
      }
     ]
    }
    UserActionMenuInputButton_Back {
     id=#6
     possessed_class=UserActionMenuInputButton_C
     tracks [{
       type=MovieSceneFloatTrack
       display_name=Render Opacity
       sections [{type=MovieSceneFloatSection, end_frame=5, start_frame=0, channels [None {type=MovieSceneScriptingFloatChannel, keys [[2,0],[5,1]]}]}]
      }]
    }
    Image_OpenArrow {
     id=#7
     possessed_class=Image
     tracks [
      {
       type=MovieSceneFloatTrack
       display_name=Render Opacity
       sections [{type=MovieSceneFloatSection, end_frame=5, start_frame=0, channels [None {type=MovieSceneScriptingFloatChannel, keys [[0,1],[5,0]]}]}]
      }
      {
       type=MovieScene2DTransformTrack
       sections [{
         type=MovieScene2DTransformSection
         end_frame=5
         start_frame=0
         channels [
          Translation.X {type=MovieSceneScriptingFloatChannel, keys [[0,0],[5,10]]}
          Scale.X {type=MovieSceneScriptingFloatChannel, keys [[0,1],[5,1.2]]}
          Scale.Y {type=MovieSceneScriptingFloatChannel, keys [[0,1],[5,1.2]]}
         ]
        }]
      }
     ]
    }
    UserActionMenuInputButton_Accept {
     id=#8
     possessed_class=UserActionMenuInputButton_C
     tracks [{
       type=MovieSceneFloatTrack
       display_name=Render Opacity
       sections [{type=MovieSceneFloatSection, end_frame=5, start_frame=0, channels [None {type=MovieSceneScriptingFloatChannel, keys [[0,1],[5,0]]}]}]
      }]
    }
   ]}
 }
]
graphs [
 BP_GetDesiredFocusTarget {nodes [
   Get Desired Focus Target()
   > Return Node!(execute, ReturnValue:Widget Object Reference=Self-Reference())
  ]}
 Close {nodes [
   Close() {is=title K2Node_FunctionEntry}
   > Branch(Condition=Get IsOpen() -> (IsOpen:bool))
   > SetVisibility(
    self=[
     Get Image_OpenArrow() -> (Image_OpenArrow:Image Object Reference)
     Get UserActionMenuInputButton_Accept() -> (UserActionMenuInputButton_Accept:User Action Menu Input Button Object Reference)
    ]
    InVisibility=SelfHitTestInvisible
   )
   > PlayAnimation(InAnimation=Get ActionCancelled() -> (ActionCancelled:Widget Animation Object Reference))
   > Set IsOpen(IsOpen:bool=false) -> (Output_Get:bool)
   > RequestCloseDialog!(execute, self:Creative Property Edit Widget Base Object Reference, bSetFocusToOptionsList:bool=true) -> (then)
   > n1 = Get!(execute:Exec) -> (then:Exec, else:Exec, IconPicker:Creative Property Editor Icon Picker Object Reference)
   > Close(self:Fort Creative Property Editor Dialog Base Object Reference=@n1) {is=title K2Node_CallFunction}
  ]}
 EventGraph {nodes [
   "On Button Clicked (UserActionMenuInputButton_Back)"!() -> (OutputDelegate, then, Button:Common Button Legacy Object Reference) {is=title K2Node_ComponentBoundEvent}
   > Close(self:Self Object Reference)
   "On Button Clicked (UserActionMenuInputButton_Accept)"!() -> (OutputDelegate, then, Button:Common Button Legacy Object Reference) {is=title K2Node_ComponentBoundEvent}
   > n1 = Branch(Condition=Get IsOpen())
   Open(execute=@n1.else)
   > Is Valid(InputObject=Get IconPicker() -> (IconPicker:Creative Property Editor Icon Picker Object Reference))
   > SelectSwatch!(
    execute
    self:Creative Property Editor Icon Picker Object Reference=Get IconPicker()
    Swatch Index:int=Get CurrentIndexInStepperValue() -> (CurrentIndexInStepperValue:int=0)
   ) -> (then)
   "On Pressed (TouchButton)"() {is=title K2Node_ComponentBoundEvent}
   > Branch(Condition=IsUsingTouch(PlayerController=GetOwningPlayer()))
   > n2 = Branch(Condition=Get IsOpen())
   > Close()
   Open(execute=@n2.else)
   > SelectSwatch(self=Get IconPicker(), Swatch Index=Get CurrentIndexInStepperValue() -> (CurrentIndexInStepperValue=0))
   ConfirmAnimationFinished() {is=title K2Node_CustomEvent}
   > SetVisibility(self=[Get Image_OpenArrow(), Get UserActionMenuInputButton_Accept()], InVisibility=Hidden)
   > SetVisibility(
    self=[
     Get UserActionMenuInputButton_Back() -> (UserActionMenuInputButton_Back:User Action Menu Input Button Object Reference)
     Get Image_CloseArrow() -> (Image_CloseArrow:Image Object Reference)
    ]
    InVisibility=SelfHitTestInvisible
   )
   CancelAnimationFinished() {is=title K2Node_CustomEvent}
   > SetVisibility(self=[Get UserActionMenuInputButton_Back(), Get Image_CloseArrow()], InVisibility=Hidden)
   > SetVisibility(self=[Get UserActionMenuInputButton_Accept(), Get Image_OpenArrow()], InVisibility=SelfHitTestInvisible)
   n3 = OnSwatchCommited!() -> (OutputDelegate, then, Index:int, DisplayText:Text, IconValue:Creative Icon Option Structure) {is=title K2Node_CustomEvent}
   > Set CurrentIndexInStepperValue(CurrentIndexInStepperValue:int=@n3.Index) -> (Output_Get:int=0)
   > SetIsModified!(execute, self:Creative Property Edit Widget Base Object Reference, bModified:bool=false) -> (then)
   > SetIsModified(bModified=true)
   > Close()
   > Update Selected Name and Swatch()
   n4 = Event BP_OnModified!() -> (OutputDelegate, then, bModified:bool) {is=title K2Node_Event}
   > UpdateModifiedState!(execute, self, isModifiedFromDefaultValue:bool=@n4.bModified) -> (then)
   n5 = Event SetDesiredValueWidth!() -> (OutputDelegate, then, DesiredValueWidth:float) {is=title K2Node_Event}
   > SetWidthOverride(self=Get SizeBox_Value() -> (SizeBox_Value:Size Box Object Reference), InWidthOverride=float + float(A=@n5.DesiredValueWidth, B=20))
   > SetWidthOverride(self=Get SizeBox_ColorName() -> (SizeBox_ColorName:Size Box Object Reference), InWidthOverride=float - float(A=@n5.DesiredValueWidth, B=120))
   Event Destruct()
   > Is Valid(InputObject=Get IconPicker())
   > Unbind Event from On Swatch Selection Committed(self:Creative Property Editor Icon Picker Object Reference=Get IconPicker(), Delegate=Create Event()) {is=title K2Node_RemoveDelegate}
   > Unbind Event from On Icon Picker Closed(self:Creative Property Editor Icon Picker Object Reference=Get IconPicker(), Delegate=Create Event()) {is=title K2Node_RemoveDelegate}
   > UnbindFromAnimationFinished!(
    execute
    self:User Widget Object Reference
    Animation:Widget Animation Object Reference=Get ActionConfirmed() -> (ActionConfirmed:Widget Animation Object Reference)
    Delegate:Delegate=Create Event()
   ) -> (then)
   > UnbindFromAnimationFinished(Animation=Get ActionCancelled(), Delegate=Create Event())
   Event BP_OnEditWidgetClicked() {is=title K2Node_Event}
   > n6 = Branch(Condition=Get IsOpen())
   > Close()
   Open(execute=@n6.else)
   > SelectSwatch(self=Get IconPicker(), Swatch Index=Get CurrentIndexInStepperValue() -> (CurrentIndexInStepperValue=0))
   Event BP_PostInitEditWidget() {is=title K2Node_Event}
   > n7 = Sequence!(execute) -> (then_0, then_1, then_2:Exec, then_3:Exec, then_4:Exec)
   Is Valid(exec=@n7.then_2, InputObject=Get IconPicker())
   > Bind Event to On Swatch Selection Committed(self:Creative Property Editor Icon Picker Object Reference=Get IconPicker(), Delegate=Create Event()) {is=title K2Node_AddDelegate}
   > Bind Event to On Icon Picker Closed(self:Creative Property Editor Icon Picker Object Reference=Get IconPicker(), Delegate=Create Event()) {is=title K2Node_AddDelegate}
   > BindToAnimationFinished(Animation=Get ActionConfirmed(), Delegate=Create Event())
   > BindToAnimationFinished(Animation=Get ActionCancelled(), Delegate=Create Event())
   > n8 = For Each Loop(Array:Array of Icon Option Object Wrapper Object References=Get IconOptionObjects() -> (IconOptionObjects:Array of Icon Option Object Wrapper Object References)) -> (Array Element:Icon Option Object Wrapper Object Reference)
   > AddOption!(execute, self:Fort Creative Property Editor Dialog Base Object Reference=Get IconPicker(), Object:Object Reference=@n8) -> (then)
   n9 = Cast To CreativeIconPickerSwatch!(execute=@n7.then_3, Object=Get ExtraContentWidget() -> (ExtraContentWidget:User Widget Object Reference)) -> (then, CastFailed, AsCreative Icon Picker Swatch:Creative Icon Picker Swatch Object Reference) {is=title K2Node_DynamicCast}
   > n10 = Set ExtraIconContentSwatch(ExtraIconContentSwatch:Creative Icon Picker Swatch Object Reference=@n9) -> (Output_Get:Creative Icon Picker Swatch Object Reference)
   UpdateText!(
    execute=@n7.then_4
    self:WBP Creative Property Editor Content Primary Object Reference=Get WBP_Content_Primary() -> (WBP_Content_Primary:WBP Creative Property Editor Content Primary Object Reference)
    Text:"Text (by ref)"=GetEntryDisplayName!(self:Creative Property Edit Widget Base Object Reference) -> (ReturnValue:Text)
   ) -> (then)
   > Update Selected Name and Swatch()
   n11 = Cast To CreativePropertyEditorIconPicker!(execute=@n7.then_0, Object=Get DialogWidget() -> (DialogWidget:Fort Creative Property Editor Dialog Base Object Reference)) -> (then, CastFailed, AsCreative Property Editor Icon Picker:Creative Property Editor Icon Picker Object Reference) {is=title K2Node_DynamicCast}
   > Set IconPicker(IconPicker:Creative Property Editor Icon Picker Object Reference=@n11) -> (Output_Get:Creative Property Editor Icon Picker Object Reference)
   SetVisibility(execute=@n7.then_1, self=[Get Image_OpenArrow(), Get UserActionMenuInputButton_Accept()], InVisibility=SelfHitTestInvisible)
   > CancelAnimationFinished() {is=title K2Node_CallFunction}
   > SetVisibility(self=Get DialogWidget(), InVisibility=Collapsed)
   SetPadding!(execute, self:User Widget Object Reference=@n10, InPadding:Margin Structure=Make Margin(Left=50, Top=0, Right=0, Bottom=0)) -> (then)
  ]}
 OnFocusReceived {nodes [
   OnFocusReceived!() -> (then, MyGeometry:Geometry Structure, InFocusEvent:Focus Event Structure) {is=title K2Node_FunctionEntry}
   > SetKeyboardFocus(self:Widget Object Reference)
   > Return Node(ReturnValue:Event Reply Structure=Handled())
  ]}
 Open {nodes [
   Open() {is=title K2Node_FunctionEntry}
   > SetVisibility(self=[Get UserActionMenuInputButton_Back(), Get Image_CloseArrow()], InVisibility=SelfHitTestInvisible)
   > PlayAnimation(InAnimation=Get ActionConfirmed())
   > Set IsOpen(IsOpen=true)
   > n1 = Get()
   > Open(self:Fort Creative Property Editor Dialog Base Object Reference=@n1) {is=title K2Node_CallFunction}
  ]}
 Update Selected Name and Swatch {nodes [
   Update Selected Name and Swatch() {is=title K2Node_FunctionEntry}
   n1 = Get CurrentIndexInStepperValue() -> (CurrentIndexInStepperValue=0)
   n2 = "Get (a copy)"(
    Array:Array of Icon Option Object Wrapper Object References=Get IconOptionObjects()
    Dimension 1=SelectInt(A=0, B=@n1, bPickA="integer < integer"(A=@n1))
   ) -> (Output:Icon Option Object Wrapper Object Reference)
   > Is Valid(InputObject=@n2)
   n3 = Get IconOption!(self:Icon Option Object Wrapper Object Reference=@n2) -> (IconOption:Playlist Option Icon Value Structure)
   > Set Represented Icon!(
    execute
    self:Creative Icon Picker Swatch Object Reference=[
     Get CreativeIconPickerSwatch() -> (CreativeIconPickerSwatch:Creative Icon Picker Swatch Object Reference)
     Get ExtraIconContentSwatch() -> (ExtraIconContentSwatch:Creative Icon Picker Swatch Object Reference)
    ]
    Icon Option:"Playlist Option Icon Value Structure (by ref)"=@n3
   ) -> (then) {is=title K2Node_CallFunction}
   > "SetText (Text)"(
    self=Get Text_SelectedColor() -> (Text_SelectedColor:Common Text Object Reference)
    InText=TextToUpper!(InText:"Text (by ref)"=Break Playlist Option Icon Value!(PlaylistOptionIconValue:"Playlist Option Icon Value Structure (by ref)"=@n3) -> (Value:Creative Icon Option Structure, DisplayName:Text, OptionValueName:String, GrantedContextTags:Gameplay Tag Container Structure) {is=title K2Node_BreakStruct}.DisplayName) -> (ReturnValue:Text)
   )
  ]}
 UpdateModifiedState {nodes [
   n1 = UpdateModifiedState!() -> (then, isModifiedFromDefaultValue:bool=~) {is=title K2Node_FunctionEntry}
   > Event Update IsModified!(execute, self:WBP Creative Property Editor Content Primary Object Reference=Get WBP_Content_Primary(), bIsModified:bool=@n1) -> (then)
  ]}
]
variables [
 CanvasSlot {type=CanvasPanelSlot, default=null, class=~, properties=~}
 ExtraIconContentSwatch {type=CreativeIconPickerSwatch_C, default=null}
 IconPicker {type=CreativePropertyEditorIconPicker_C, default=null}
 IsOpen {type=boolean}
]
widgets {
 info {
  inheritedWidgetCount=1
  parentClass=/Script/FortniteUI.CreativePropertyEditWidgetIconBase
  rootWidgetClass=&Overlay
  widgetCount=25
 }
 widgets [
  Overlay_25 {widgetClassPath=&Overlay, widget=Overlay_25}
  SizeBox_2 {
   widgetClassPath=&SizeBox
   parent=Overlay_25
   widget=SizeBox_2
   properties {bOverride_HeightOverride=true, heightOverride=82}
   slot {class=OverlaySlot, properties {horizontalAlignment=HAlign_Fill, verticalAlignment=VAlign_Center, padding=<ltrb 16 0 16 0>}}
  }
  Overlay_0 {widgetClassPath=&Overlay, parent=SizeBox_2, widget=Overlay_0, slot {class=SizeBoxSlot}}
  WBP_Background_SkewedWithMarker {
   widgetClassPath=/Game/Athena/UI/InGame/Creative/PropertyEditor/WBP_CreativePropertyEditor_Background_SkewedWithMarker
   bIsVariable=true
   parent=Overlay_0
   widget=WBP_Background_SkewedWithMarker
   properties {focusState=Hovered}
   slot {class=OverlaySlot, properties {horizontalAlignment=HAlign_Fill, verticalAlignment=VAlign_Fill}}
  }
  SaveSpinner_2 {
   widgetClassPath=&Image
   bIsVariable=true
   parent=Overlay_0
   widget=SaveSpinner_2
   properties {
    visibility=Collapsed
    brush {
     bIsDynamicallyLoaded=false
     drawAs=Image
     imageType=NoImage
     mirroring=NoMirror
     resourceName=None
     resourceObject=/FortUILibrary/UIKit/Throbber/MI_UIKit_Monochrome_Throbber
     tiling=NoTile
     imageSize=<24 24>
     margin=<ltrb 0 0 0 0>
     outlineSettings {
      bUseBrushTransparency=false
      roundingType=HalfHeightRadius
      width=0
      color {colorUseRule=UseColor_Specified, specifiedColor=<rgba 0 0 0 0>}
      cornerRadii=<0 0 0 0>
     }
     tintColor {colorUseRule=UseColor_Specified, specifiedColor=<rgba 1 1 1 1>}
     uVRegion {bIsValid=false, max=<0 0>, min=<0 0>}
    }
    colorAndOpacity=<rgba 0 0.2961 0.9647 1>
   }
   slot {class=OverlaySlot, properties {verticalAlignment=VAlign_Center, padding=<ltrb 10 -24 0 0>}}
  }
  TouchButton {
   widgetClassPath=/Script/FortniteUI.FortClickThroughButton
   bIsVariable=true
   parent=Overlay_0
   widget=TouchButton
   properties {widgetStyle {
     clickedSlateSound {resourceObject=null}
     disabled {
      bIsDynamicallyLoaded=false
      drawAs=NoDrawType
      imageType=NoImage
      mirroring=NoMirror
      resourceName=None
      resourceObject=null
      tiling=NoTile
      imageSize=<0 0>
      margin=<ltrb 0 0 0 0>
      outlineSettings {
       bUseBrushTransparency=false
       roundingType=HalfHeightRadius
       width=0
       color {colorUseRule=UseColor_Specified, specifiedColor=<rgba 0 0 0 0>}
       cornerRadii=<0 0 0 0>
      }
      tintColor {colorUseRule=UseColor_Specified, specifiedColor=<rgba 1 1 1 1>}
      uVRegion {bIsValid=false, max=<0 0>, min=<0 0>}
     }
     disabledForeground {colorUseRule=UseColor_Specified, specifiedColor=<rgba 0.5271 0.5271 0.5271 1>}
     hovered {
      bIsDynamicallyLoaded=false
      drawAs=NoDrawType
      imageType=NoImage
      mirroring=NoMirror
      resourceName=None
      resourceObject=null
      tiling=NoTile
      imageSize=<32 32>
      margin=<ltrb 0 0 0 0>
      outlineSettings {
       bUseBrushTransparency=true
       roundingType=FixedRadius
       width=1
       color {colorUseRule=UseColor_Specified, specifiedColor=<rgba 0.7243 0.7243 0.7243 1>}
       cornerRadii=<4 4 4 4>
      }
      tintColor {colorUseRule=UseColor_Specified, specifiedColor=<rgba 0.7243 0.7243 0.7243 1>}
      uVRegion {bIsValid=false, max=<0 0>, min=<0 0>}
     }
     hoveredForeground {colorUseRule=UseColor_Specified, specifiedColor=<rgba 0 1 1 1>}
     hoveredSlateSound {resourceObject=null}
     normal {
      bIsDynamicallyLoaded=false
      drawAs=NoDrawType
      imageType=NoImage
      mirroring=NoMirror
      resourceName=None
      resourceObject=null
      tiling=NoTile
      imageSize=<32 32>
      margin=<ltrb 0 0 0 0>
      outlineSettings {
       bUseBrushTransparency=true
       roundingType=FixedRadius
       width=1
       color {colorUseRule=UseColor_Specified, specifiedColor=<rgba 0.6951 0.6951 0.6951 1>}
       cornerRadii=<4 4 4 4>
      }
      tintColor {colorUseRule=UseColor_Specified, specifiedColor=<rgba 0.4955 0.4955 0.4955 1>}
      uVRegion {bIsValid=false, max=<0 0>, min=<0 0>}
     }
     normalForeground {colorUseRule=UseColor_Specified, specifiedColor=<rgba 0 1 1 1>}
     normalPadding=<ltrb 0 0 0 0>
     pressed {
      bIsDynamicallyLoaded=false
      drawAs=NoDrawType
      imageType=NoImage
      mirroring=NoMirror
      resourceName=None
      resourceObject=null
      tiling=NoTile
      imageSize=<32 32>
      margin=<ltrb 0 0 0 0>
      outlineSettings {
       bUseBrushTransparency=true
       roundingType=FixedRadius
       width=1
       color {colorUseRule=UseColor_Specified, specifiedColor=<rgba 0.7243 0.7243 0.7243 1>}
       cornerRadii=<4 4 4 4>
      }
      tintColor {colorUseRule=UseColor_Specified, specifiedColor=<rgba 0.3843 0.3843 0.3843 1>}
      uVRegion {bIsValid=false, max=<0 0>, min=<0 0>}
     }
     pressedForeground {colorUseRule=UseColor_Specified, specifiedColor=<rgba 0 1 1 1>}
     pressedPadding=<ltrb 0 0 0 0>
     pressedSlateSound {resourceObject=null}
    }}
   slot {class=OverlaySlot, properties {horizontalAlignment=HAlign_Fill, verticalAlignment=VAlign_Fill}}
  }
  HorizontalBox_2 {
   widgetClassPath=&HorizontalBox
   parent=Overlay_0
   widget=HorizontalBox_2
   slot {class=OverlaySlot, properties {horizontalAlignment=HAlign_Fill, verticalAlignment=VAlign_Fill}}
  }
  WBP_Content_Primary {
   widgetClassPath=/Game/Athena/UI/InGame/Creative/PropertyEditor/WBP_CreativePropertyEditor_Content_Primary
   bIsVariable=true
   parent=HorizontalBox_2
   widget=WBP_Content_Primary
   properties {richText="[Parameter Name]", states=DefaultHovered}
   slot {class=HorizontalBoxSlot, properties {verticalAlignment=VAlign_Center, padding=<ltrb 30 0 4 0>, size {sizeRule=Fill, value=1}}}
  }
  SizeBox_Value {
   widgetClassPath=&SizeBox
   bIsVariable=true
   parent=HorizontalBox_2
   widget=SizeBox_Value
   properties {bOverride_WidthOverride=true, widthOverride=340}
   slot {class=HorizontalBoxSlot, properties {verticalAlignment=VAlign_Center, padding=<ltrb 0 0 15 0>}}
  }
  Overlay_2 {widgetClassPath=&Overlay, parent=SizeBox_Value, widget=Overlay_2, slot {class=SizeBoxSlot}}
  OptionSelectionArea_2 {
   widgetClassPath=&HorizontalBox
   bIsVariable=true
   parent=Overlay_2
   widget=OptionSelectionArea_2
   slot {class=OverlaySlot, properties {horizontalAlignment=HAlign_Fill}}
  }
  SizeBox_ColorName {
   widgetClassPath=&SizeBox
   bIsVariable=true
   parent=Overlay_2
   widget=SizeBox_ColorName
   properties {
    bOverride_HeightOverride=true
    bOverride_WidthOverride=true
    clipping=ClipToBounds
    heightOverride=50
    maxDesiredWidth=816
    widthOverride=200
   }
   slot {class=OverlaySlot, properties {verticalAlignment=VAlign_Center}}
  }
  Text_SelectedColor {
   widgetClassPath=&CommonTextBlock
   bIsVariable=true
   parent=SizeBox_ColorName
   widget=Text_SelectedColor
   properties {
    justification=Center
    scrollStyle=&TextScrollStyle-Base_C
    text=SANDSTONE YELLOW
    colorAndOpacity {colorUseRule=UseColor_Specified, specifiedColor=<rgba 0 0 0 1>}
    font {
     bForceMonospaced=false
     bMaterialIsStencil=false
     fontMaterial=null
     fontObject=&BurbankSmall
     letterSpacing=0
     monospacedWidth=1
     size=20
     skewAmount=0
     typefaceFontName=Bold
     outlineSettings {
      bApplyOutlineToDropShadows=false
      bMiteredCorners=false
      bSeparateFillAlpha=false
      outlineMaterial=null
      outlineSize=0
      outlineColor=<rgba 0 0.0182 0.2031 0.2>
     }
    }
    shadowColorAndOpacity=<rgba 0.0052 0.0052 0.0056 0>
    shadowOffset=<2 1>
   }
   slot {class=SizeBoxSlot, properties {horizontalAlignment=HAlign_Right, verticalAlignment=VAlign_Center, padding=<ltrb 10 0 8 0>}}
  }
  HorizontalBox_0 {
   widgetClassPath=&HorizontalBox
   parent=Overlay_2
   widget=HorizontalBox_0
   slot {class=OverlaySlot, properties {horizontalAlignment=HAlign_Right, verticalAlignment=VAlign_Center}}
  }
  Overlay_3 {
   widgetClassPath=&Overlay
   parent=HorizontalBox_0
   widget=Overlay_3
   slot {class=HorizontalBoxSlot, properties {horizontalAlignment=HAlign_Right, verticalAlignment=VAlign_Center, padding=<ltrb 0 0 5 0>}}
  }
  ColorSwatchShadow {
   widgetClassPath=&Image
   bIsVariable=true
   parent=Overlay_3
   widget=ColorSwatchShadow
   properties {
    visibility=SelfHitTestInvisible
    brush {
     bIsDynamicallyLoaded=false
     drawAs=Image
     imageType=NoImage
     mirroring=NoMirror
     resourceName=None
     resourceObject=null
     tiling=NoTile
     imageSize=<64 64>
     margin=<ltrb 0 0 0 0>
     outlineSettings {
      bUseBrushTransparency=false
      roundingType=HalfHeightRadius
      width=0
      color {colorUseRule=UseColor_Specified, specifiedColor=<rgba 0 0 0 0>}
      cornerRadii=<0 0 0 0>
     }
     tintColor {colorUseRule=UseColor_Specified, specifiedColor=<rgba 1 1 1 1>}
     uVRegion {bIsValid=false, max=<0 0>, min=<0 0>}
    }
    colorAndOpacity=<rgba 0 0 0 0.5>
    renderTransform {angle=0, scale=<1 1>, shear=<0 0>, translation=<0 4>}
   }
   slot {class=OverlaySlot, properties {horizontalAlignment=HAlign_Right, verticalAlignment=VAlign_Bottom}}
  }
  SizeBox_261 {
   widgetClassPath=&SizeBox
   parent=Overlay_3
   widget=SizeBox_261
   properties {bOverride_HeightOverride=true, bOverride_WidthOverride=true, heightOverride=64, widthOverride=64}
   slot {class=OverlaySlot}
  }
  CreativeIconPickerSwatch {
   widgetClassPath=/Game/Athena/UI/InGame/Creative/PropertyEditor/Icon_Picker/CreativeIconPickerSwatch
   bIsVariable=true
   parent=SizeBox_261
   widget=CreativeIconPickerSwatch
   slot {class=SizeBoxSlot, properties {horizontalAlignment=HAlign_Left, verticalAlignment=VAlign_Top}}
  }
  Overlay_4 {widgetClassPath=&Overlay, parent=HorizontalBox_0, widget=Overlay_4, slot {class=HorizontalBoxSlot}}
  UserActionMenuInputButton_Accept {
   widgetClassPath=/Game/UI/SocialPanel/UserActionMenuInputButton
   bIsVariable=true
   parent=Overlay_4
   widget=UserActionMenuInputButton_Accept
   properties {
    bHideInputActionWithKeyboard=true
    onButtonBaseClicked=(&self:WidgetTree.UserActionMenuInputButton_Accept.HandleOnButtonClicked)
    onButtonBaseDoubleClicked=(&self:WidgetTree.UserActionMenuInputButton_Accept.HandleOnButtonDoubleClicked)
    onButtonBaseHovered=(&self:WidgetTree.UserActionMenuInputButton_Accept.HandleOnButtonHovered)
    onButtonBaseUnhovered=(&self:WidgetTree.UserActionMenuInputButton_Accept.HandleOnButtonUnhovered)
    onSelectedChangedBase=(&self:WidgetTree.UserActionMenuInputButton_Accept.HandleOnSelectedChanged)
    style=/FortUILibrary/Foundation/Buttons/ButtonStyles/Athena/ButtonStyle_TransparentList_Social_NoHovered.ButtonStyle_TransparentList_Social_NoHovered_C
    triggeringInputAction {dataTable=&InputActionDataTable, rowName=GenericForward}
   }
   slot {class=OverlaySlot}
  }
  Image_CloseArrow {
   widgetClassPath=&Image
   bIsVariable=true
   parent=Overlay_4
   widget=Image_CloseArrow
   properties {
    brush {
     bIsDynamicallyLoaded=false
     drawAs=Image
     imageType=NoImage
     mirroring=NoMirror
     resourceName=None
     resourceObject=/Game/UI/SocialPanel/Art/Textures/T_UI_Icon_TreeArrow_64
     tiling=NoTile
     imageSize=<30 30>
     margin=<ltrb 0 0 0 0>
     outlineSettings {
      bUseBrushTransparency=false
      roundingType=HalfHeightRadius
      width=0
      color {colorUseRule=UseColor_Specified, specifiedColor=<rgba 0 0 0 0>}
      cornerRadii=<0 0 0 0>
     }
     tintColor {colorUseRule=UseColor_Specified, specifiedColor=<rgba 1 1 1 1>}
     uVRegion {bIsValid=false, max=<0 0>, min=<0 0>}
    }
    colorAndOpacity=<rgba 0.25 0.25 0.25 1>
    renderTransform {angle=0, scale=<-1 1>, shear=<0 0>, translation=<0 0>}
   }
   slot {class=OverlaySlot, properties {verticalAlignment=VAlign_Center}}
  }
  Overlay_1 {
   widgetClassPath=&Overlay
   parent=HorizontalBox_0
   widget=Overlay_1
   slot {class=HorizontalBoxSlot, properties {horizontalAlignment=HAlign_Right, verticalAlignment=VAlign_Center}}
  }
  UserActionMenuInputButton_Back {
   widgetClassPath=&UserActionMenuInputButton
   bIsVariable=true
   parent=Overlay_1
   widget=UserActionMenuInputButton_Back
   properties {
    bHideInputActionWithKeyboard=true
    bStopAction=true
    inputPriority=10
    onButtonBaseClicked=(&self:WidgetTree.UserActionMenuInputButton_Back.HandleOnButtonClicked)
    onButtonBaseDoubleClicked=(&self:WidgetTree.UserActionMenuInputButton_Back.HandleOnButtonDoubleClicked)
    onButtonBaseHovered=(&self:WidgetTree.UserActionMenuInputButton_Back.HandleOnButtonHovered)
    onButtonBaseUnhovered=(&self:WidgetTree.UserActionMenuInputButton_Back.HandleOnButtonUnhovered)
    onSelectedChangedBase=(&self:WidgetTree.UserActionMenuInputButton_Back.HandleOnSelectedChanged)
    priority=10
    style=&ButtonStyle_TransparentList_Social_NoHovered_C
    visibility=Collapsed
    triggeringInputAction {dataTable=&InputActionDataTable, rowName=GenericBack}
   }
   slot {class=OverlaySlot, properties {horizontalAlignment=HAlign_Fill, verticalAlignment=VAlign_Fill}}
  }
  Image_OpenArrow {
   widgetClassPath=&Image
   bIsVariable=true
   parent=Overlay_1
   widget=Image_OpenArrow
   properties {
    brush {
     bIsDynamicallyLoaded=false
     drawAs=Image
     imageType=NoImage
     mirroring=NoMirror
     resourceName=None
     resourceObject=&T_UI_Icon_TreeArrow_64
     tiling=NoTile
     imageSize=<30 30>
     margin=<ltrb 0 0 0 0>
     outlineSettings {
      bUseBrushTransparency=false
      roundingType=HalfHeightRadius
      width=0
      color {colorUseRule=UseColor_Specified, specifiedColor=<rgba 0 0 0 0>}
      cornerRadii=<0 0 0 0>
     }
     tintColor {colorUseRule=UseColor_Specified, specifiedColor=<rgba 1 1 1 1>}
     uVRegion {bIsValid=false, max=<0 0>, min=<0 0>}
    }
    colorAndOpacity=<rgba 0 0.1441 1 1>
   }
   slot {class=OverlaySlot, properties {horizontalAlignment=HAlign_Right, verticalAlignment=VAlign_Center}}
  }
  Button_Collapse {
   widgetClassPath=&IconTextButton
   bInherited=true
   parent=Overlay_25
   widget=Button_Collapse
   properties {
    onButtonBaseClicked=(&self:WidgetTree.Button_Collapse.HandleOnButtonClicked)
    onButtonBaseDoubleClicked=(&self:WidgetTree.Button_Collapse.HandleOnButtonDoubleClicked)
    onButtonBaseHovered=(&self:WidgetTree.Button_Collapse.HandleOnButtonHovered)
    onButtonBaseUnhovered=(&self:WidgetTree.Button_Collapse.HandleOnButtonUnhovered)
    onSelectedChangedBase=(&self:WidgetTree.Button_Collapse.HandleOnSelectedChanged)
    visibility=Collapsed
   }
   slot {class=OverlaySlot, properties {horizontalAlignment=HAlign_Center, verticalAlignment=VAlign_Center}}
  }
 ]
}
