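# Creative property editor row widget for basic string properties, pairing a label with a filtered text input field, derived from CreativePropertyEditWidgetStringBase.
# Text changes set modified when non-empty and request dialog close; BP_OnModified forwards the state to the primary content widget, which also shows the entry name.
# Receiving focus centers and keyboard-focuses the field.
# Value width follows the requested width plus padding, and delayed initialization recolors the input's dynamic font material.
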WidgetBlueprint /Game/Athena/UI/InGame/Creative/PropertyEditor/EditWidgets/CreativePropertyEditWidgetBase_String_Basic
graphs [
 BP_GetDesiredFocusTarget {nodes [
   Get Desired Focus Target()
   > Return Node!(
    execute
    ReturnValue:Widget Object Reference=Get FilteredTextEntry_Widget() -> (FilteredTextEntry_Widget:Fort Editable Filtered Counted Text Box Object Reference)
   )
  ]}
 EventGraph {nodes [
   n1 = "On Text Changed (FilteredTextEntry_Widget)"!() -> (OutputDelegate, then, Text:Text) {is=title K2Node_ComponentBoundEvent}
   > SetIsModified!(execute, self:Creative Property Edit Widget Base Object Reference, bModified:bool=NOT Boolean(A=TextIsEmpty(InText=@n1.Text))) -> (then)
   > RequestCloseDialog!(execute, self:Creative Property Edit Widget Base Object Reference, bSetFocusToOptionsList:bool=false) -> (then)
   n2 = Event BP_OnModified!() -> (OutputDelegate, then, bModified:bool) {is=title K2Node_Event}
   > UpdateModifiedState!(execute, self, isModifiedFromDefaultValue:bool=@n2.bModified) -> (then)
   Event OnInitialized()
   > Delay(Duration=0.2)
   > n3 = Cast To WBP_UIKit_FilteredInputField!(execute, Object=Get FilteredTextEntry_Widget()) -> (then, CastFailed, AsWBP UIKit Filtered Input Field:WBP UIKit Filtered Input Field Object Reference) {is=title K2Node_DynamicCast}
   n4 = Get FontDynamicMaterial!(self:WBP UIKit Block Editable Text Object Reference=Get Block_EditableText!(self:WBP UIKit Input Field Base Object Reference=Get InputField!(self:WBP UIKit Filtered Input Field Object Reference=@n3) -> (InputField:WBP UIKit Input Field Object Reference)) -> (Block_EditableText:WBP UIKit Block Editable Text Object Reference)) -> (FontDynamicMaterial:Material Instance Dynamic Object Reference)
   > SetVectorParameterValue(self=@n4, ParameterName=Color_Idle, Value=(R=0.0578,G=0.0578,B=0.0578,A=1))
   > SetVectorParameterValue(self=@n4, ParameterName=Color_Hovered, Value=(R=0.0931,G=0.0931,B=0.0931,A=1))
   > SetVectorParameterValue(self=@n4, ParameterName=Color_Focused, Value=(R=0.0931,G=0.0931,B=0.0931,A=1))
   > SetVectorParameterValue(self=@n4, ParameterName=Color_Pressed, Value=(R=0.0931,G=0.0931,B=0.0931,A=1))
   n5 = Event SetDesiredValueWidth!() -> (OutputDelegate, then, DesiredValueWidth:float) {is=title K2Node_Event}
   > SetWidthOverride(self=Get SizeBox_Value() -> (SizeBox_Value:Size Box Object Reference), InWidthOverride=float + float(A=@n5.DesiredValueWidth, B=47))
   Event BP_PostInitEditWidget() {is=title K2Node_Event}
   > UpdateText!(
    execute
    self:WBP Creative Property Editor Content Primary Object Reference=Get WBP_Content_Primary() -> (WBP_Content_Primary:WBP Creative Property Editor Content Primary Object Reference)
    Text:"Text (by ref)"=GetEntryDisplayName!(self:Creative Property Edit Widget Base Object Reference) -> (ReturnValue:Text)
   ) -> (then)
  ]}
 OnFocusReceived {nodes [
   OnFocusReceived!() -> (then, MyGeometry:Geometry Structure, InFocusEvent:Focus Event Structure) {is=title K2Node_FunctionEntry}
   n1 = Get FilteredTextEntry_Widget()
   > CenterWidget!(execute, ToCenter:Widget Object Reference=@n1) -> (then)
   > SetKeyboardFocus(self:Widget Object Reference=@n1)
   > Return Node(ReturnValue:Event Reply Structure=Handled())
  ]}
 UpdateModifiedState {nodes [
   n1 = UpdateModifiedState!() -> (then, isModifiedFromDefaultValue:bool=~) {is=title K2Node_FunctionEntry}
   > Event Update IsModified!(execute, self:WBP Creative Property Editor Content Primary Object Reference=Get WBP_Content_Primary(), bIsModified:bool=@n1) -> (then)
  ]}
]
variables []
widgets {
 info {
  inheritedWidgetCount=2
  namedSlotCount=1
  parentClass=/Script/FortniteUI.CreativePropertyEditWidgetStringBase
  rootWidgetClass=&Overlay
  widgetCount=15
 }
 widgets [
  Overlay_25 {widgetClassPath=&Overlay, widget=Overlay_25}
  SizeBox_0 {
   widgetClassPath=&SizeBox
   parent=Overlay_25
   widget=SizeBox_0
   properties {bOverride_MinDesiredHeight=true, minDesiredHeight=82}
   slot {class=OverlaySlot, properties {horizontalAlignment=HAlign_Fill, verticalAlignment=VAlign_Center, padding=<ltrb 16 0 16 0>}}
  }
  Overlay_0 {widgetClassPath=&Overlay, parent=SizeBox_0, widget=Overlay_0, slot {class=SizeBoxSlot}}
  WBP_Background_SkewedWithMarker {
   widgetClassPath=/Game/Athena/UI/InGame/Creative/PropertyEditor/WBP_CreativePropertyEditor_Background_SkewedWithMarker
   bIsVariable=true
   parent=Overlay_0
   widget=WBP_Background_SkewedWithMarker
   properties {focusState=Hovered}
   slot {class=OverlaySlot, properties {horizontalAlignment=HAlign_Fill, verticalAlignment=VAlign_Fill}}
  }
  WBP_CaptureForPostBufferUpdate {
   widgetClassPath=&WBP_CaptureForPostBufferUpdate
   bIsVariable=true
   parent=Overlay_0
   slot {class=OverlaySlot, properties {horizontalAlignment=HAlign_Fill, verticalAlignment=VAlign_Fill}}
  }
  HorizontalBox_2 {
   widgetClassPath=&HorizontalBox
   parent=Overlay_0
   widget=HorizontalBox_2
   slot {class=OverlaySlot, properties {horizontalAlignment=HAlign_Fill, verticalAlignment=VAlign_Fill}}
  }
  WBP_Content_Primary {
   widgetClassPath=/Game/Athena/UI/InGame/Creative/PropertyEditor/WBP_CreativePropertyEditor_Content_Primary
   bIsVariable=true
   parent=HorizontalBox_2
   widget=WBP_Content_Primary
   properties {richText="[Parameter Name]", states=DefaultHovered}
   slot {class=HorizontalBoxSlot, properties {verticalAlignment=VAlign_Center, padding=<ltrb 30 0 4 0>, size {sizeRule=Fill, value=1}}}
  }
  SizeBox_Value {
   widgetClassPath=&SizeBox
   bIsVariable=true
   parent=HorizontalBox_2
   widget=SizeBox_Value
   properties {bOverride_WidthOverride=true, heightOverride=30, widthOverride=367}
   slot {class=HorizontalBoxSlot, properties {padding=<ltrb 0 0 32 0>, size {sizeRule=Automatic, value=0.5}}}
  }
  OptionSelectionArea_2 {widgetClassPath=&HorizontalBox, bIsVariable=true, parent=SizeBox_Value, widget=OptionSelectionArea_2, slot {class=SizeBoxSlot}}
  Overlay_1 {
   widgetClassPath=&Overlay
   parent=OptionSelectionArea_2
   widget=Overlay_1
   slot {class=HorizontalBoxSlot, properties {verticalAlignment=VAlign_Center, size {sizeRule=Fill, value=1}}}
  }
  FilteredTextEntry_Widget {
   widgetClassPath=/FortUILibrary/UIKit/InputField/WBP_UIKit_FilteredInputField
   bInherited=true
   bIsVariable=true
   parent=Overlay_1
   widget=FilteredTextEntry_Widget
   properties {bIsFocusable=true, hideValidationInfo=true, foregroundColor {colorUseRule=UseColor_Foreground, specifiedColor=<rgba 0 0 0 1>}}
   slot {class=OverlaySlot, properties {horizontalAlignment=HAlign_Fill, verticalAlignment=VAlign_Fill}}
  }
  SaveSpinner_2 {
   widgetClassPath=&Image
   bIsVariable=true
   parent=Overlay_25
   widget=SaveSpinner_2
   properties {
    visibility=Collapsed
    brush {
     bIsDynamicallyLoaded=false
     drawAs=Image
     imageType=NoImage
     mirroring=NoMirror
     resourceName=None
     resourceObject=/FortUILibrary/UIKit/Throbber/MI_UIKit_Color_Throbber
     tiling=NoTile
     imageSize=<24 24>
     margin=<ltrb 0 0 0 0>
     outlineSettings {
      bUseBrushTransparency=false
      roundingType=HalfHeightRadius
      width=0
      color {colorUseRule=UseColor_Specified, specifiedColor=<rgba 0 0 0 0>}
      cornerRadii=<0 0 0 0>
     }
     tintColor {colorUseRule=UseColor_Specified, specifiedColor=<rgba 1 1 1 1>}
     uVRegion {bIsValid=false, max=<0 0>, min=<0 0>}
    }
   }
   slot {class=OverlaySlot, properties {verticalAlignment=VAlign_Center, padding=<ltrb 10 -24 0 0>}}
  }
  SizeBox_ExpanderOpen_2 {
   widgetClassPath=&SizeBox
   bIsVariable=true
   parent=Overlay_25
   widget=SizeBox_ExpanderOpen_2
   properties {bOverride_HeightOverride=true, bOverride_WidthOverride=true, heightOverride=42, visibility=Collapsed, widthOverride=42}
   slot {class=OverlaySlot, properties {horizontalAlignment=HAlign_Right, verticalAlignment=VAlign_Center, padding=<ltrb 0 0 20 0>}}
  }
  Button_Collapse {
   widgetClassPath=/FortUILibrary/Foundation/Buttons/PanelButton
   bInherited=true
   bIsVariable=true
   parent=SizeBox_ExpanderOpen_2
   widget=Button_Collapse
   properties {
    onButtonBaseClicked=(&self:WidgetTree.Button_Collapse.HandleOnButtonClicked)
    onButtonBaseDoubleClicked=(&self:WidgetTree.Button_Collapse.HandleOnButtonDoubleClicked)
    onButtonBaseHovered=(&self:WidgetTree.Button_Collapse.HandleOnButtonHovered)
    onButtonBaseUnhovered=(&self:WidgetTree.Button_Collapse.HandleOnButtonUnhovered)
    onSelectedChangedBase=(&self:WidgetTree.Button_Collapse.HandleOnSelectedChanged)
    style=/FortUILibrary/Foundation/Buttons/ButtonStyles/ButtonStyle-Mobile.ButtonStyle-Mobile_C
    visibility=Collapsed
    inputActionRenderTranslation=<-38 -4>
    inputActionRimBrushSize=<32 32>
    triggeringInputAction {dataTable=/FortUILibrary/Input/CreativeInputActionDataTable, rowName=CollapseEditWidget}
   }
   slot {class=SizeBoxSlot}
  }
  Image_ExpanderArrow_2 {
   widgetClassPath=&Image
   bIsVariable=true
   namedSlotHost=Button_Collapse
   widget=Image_ExpanderArrow_2
   properties {
    brush {
     bIsDynamicallyLoaded=false
     drawAs=Image
     imageType=NoImage
     mirroring=NoMirror
     resourceName=None
     resourceObject=/Game/UI/SocialPanel/Art/Textures/T_UI_Icon_TreeArrow_64
     tiling=NoTile
     imageSize=<32 32>
     margin=<ltrb 0 0 0 0>
     outlineSettings {
      bUseBrushTransparency=false
      roundingType=HalfHeightRadius
      width=0
      color {colorUseRule=UseColor_Specified, specifiedColor=<rgba 0 0 0 0>}
      cornerRadii=<0 0 0 0>
     }
     tintColor {colorUseRule=UseColor_Specified, specifiedColor=<rgba 1 1 1 1>}
     uVRegion {bIsValid=false, max=<0 0>, min=<0 0>}
    }
    renderTransform {angle=-90, scale=<1 1>, shear=<0 0>, translation=<0 0>}
   }
  }
 ]
}
