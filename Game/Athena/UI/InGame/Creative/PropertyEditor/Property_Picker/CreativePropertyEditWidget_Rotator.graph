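# Creative property editor rotator widget (CommonRotator) that cycles a setting's options with left/right arrow buttons.
# Arrow or body clicks shift the text; OnRotated plays the pressed sound only for non-mouse input.
# Pips use a dynamic material fed option count, default and selected index; one-direction mode hides arrows and restyles.
# Focus rescales arrows for mobile, mirroring the right via negative X; blocked body clicks print leftover debug text.
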WidgetBlueprint /Game/Athena/UI/InGame/Creative/PropertyEditor/Property_Picker/CreativePropertyEditWidget_Rotator
graphs [
 ChangeDirection {nodes [
   n1 = ChangeDirection!() -> (then, bDirection:bool=~) {is=title K2Node_FunctionEntry}
   > Branch(Condition=IsInteractionEnabled!(self:Common Button Base Object Reference) -> (ReturnValue:bool))
   > n2 = Branch(Condition=@n1)
   > ShiftTextLeft(self:Common Rotator Object Reference)
   ShiftTextRight(execute=@n2.else, self:Common Rotator Object Reference)
  ]}
 ConfigureForOneDirectionMode {nodes [
   n1 = ConfigureForOneDirectionMode!() -> (then, UseOneDirectionalMode:bool=~) {is=title K2Node_FunctionEntry}
   > Set OneDirectionalModeActive(OneDirectionalModeActive:bool=@n1) -> (Output_Get:bool)
   > n2 = Branch(Condition=@n1)
   > SetVisibility(
    self=[Get ButtonLeft() -> (ButtonLeft:Icon Text Button Object Reference), Get ButtonRight() -> (ButtonRight:Icon Text Button Object Reference)]
    InVisibility=Hidden
   )
   > SetStyle!(
    execute
    self:Common Text Object Reference=Get MyText() -> (MyText:Common Text Object Reference)
    InStyle:Common Text Style Class Reference=/FortUILibrary/Styles/Typography/Foundation/TextStyle-Base-S-B.TextStyle-Base-S-B_C
   ) -> (then)
   > SetStyle(
    self:Common Button Base Object Reference
    InStyle:Common Button Style Class Reference=/FortUILibrary/Foundation/Buttons/ButtonStyles/TopBar/ButtonStyle-Outline.ButtonStyle-Outline_C
   )
   > SetBrush(
    self:Border Object Reference=Get MainBorder() -> (MainBorder:Common Border Object Reference)
    InBrush=NoResourceBrush!() -> (ReturnValue:Slate Brush Structure)
   )
   SetVisibility(execute=@n2.else, self=[Get ButtonLeft(), Get ButtonRight()], InVisibility=Visible)
  ]}
 EventGraph {nodes [
   "On Button Clicked (ButtonLeft)"!() -> (OutputDelegate, then, Button:Common Button Legacy Object Reference) {is=title K2Node_ComponentBoundEvent}
   n1 = IsInteractionEnabled()
   > Branch(Condition=@n1)
   > n2 = ShiftTextLeft()
   Set bIsMouseInput(execute=[@n2.then, @n3.then], bIsMouseInput:bool=true) -> (Output_Get:bool)
   "On Button Clicked (ButtonRight)"!() -> (OutputDelegate, then, Button:Common Button Legacy Object Reference) {is=title K2Node_ComponentBoundEvent}
   > Branch(Condition=@n1)
   > n3 = ShiftTextRight()
   n4 = OnRotated_Event!() -> (OutputDelegate, then, Value:int) {is=title K2Node_CustomEvent}
   > n5 = Branch(Condition=Get bIsMouseInput() -> (bIsMouseInput:bool))
   > Set bIsMouseInput(bIsMouseInput=false)
   n6 = Cast To SoundBase!(
    execute=@n5.else
    Object=Get!(self:Common Button Style Object Reference=GetStyle!(self:Common Button Base Object Reference) -> (ReturnValue:Common Button Style Object Reference)) -> (PressedSlateSound_ResourceObject:Object Reference)
   ) -> (then, CastFailed, AsSound Base:Sound Base Object Reference) {is=title K2Node_DynamicCast}
   > PlaySound2D(Sound=@n6)
   Event On Clicked()
   > n7 = Branch(Condition=AND Boolean(A=IsInteractionEnabled(), B=Get BaseOnClickedEnabled() -> (BaseOnClickedEnabled:bool)))
   > ShiftTextRight()
   > Set bIsMouseInput(bIsMouseInput=true)
   PrintString!(execute=@n7.else, InString="fda;'dkf;'akd';fk;askdf';kas';ga", bPrintToScreen, bPrintToLog, TextColor, Duration, Key) -> (then)
   Event Construct()
   > Set bIsFocusable(bIsFocusable:bool=true) -> (Output_Get:bool)
   > SetTriggeredInputAction!(
    execute
    self:Common Button Base Object Reference=Get ButtonLeft()
    InputActionRow:"Data Table Row Handle Structure (by ref)"=Get Left Triggering Input Action() -> (Left Triggering Input Action:Data Table Row Handle Structure)
   ) -> (then)
   > SetTriggeredInputAction(
    self=Get ButtonRight()
    InputActionRow=Get Right Triggering Input Action() -> (Right Triggering Input Action:Data Table Row Handle Structure)
   )
   > ConfigureForOneDirectionMode!(execute, self, UseOneDirectionalMode:bool=false) -> (then) {is=title K2Node_CallFunction}
   > Assign On Rotated!(execute:Exec, self:Common Rotator Object Reference, Delegate:"Delegate (by ref)"=@n4) -> (then:Exec) {is=title K2Node_AssignDelegate}
  ]}
 GetRightButtonScale {nodes [
   GetRightButtonScale() {is=title K2Node_FunctionEntry}
   n1 = BreakVector2D(InVec=Get RotatorButtonsScale() -> (RotatorButtonsScale:Vector 2D Structure))
   > Return Node!(execute, RightButtonScale:Vector 2D Structure=MakeVector2D(X=float * float(A=@n1, B=-1), Y=@n1.Y))
  ]}
 InitializePips {nodes [
   n1 = InitializePips!() -> (then, PipCount:int, DoNotShowPips:bool=~) {is=title K2Node_FunctionEntry}
   > SetVisibility(
    self=Get Overlay_Pips() -> (Overlay_Pips:Overlay Object Reference)
    InVisibility=Select(Option 0=SelfHitTestInvisible, Option 1=Collapsed, Index=@n1.DoNotShowPips)
   )
   > n2 = GetDynamicMaterial(self=Get PipIndicator() -> (PipIndicator:Image Object Reference))
   > Set PipsDynamicMaterial(PipsDynamicMaterial:Material Instance Dynamic Object Reference=@n2) -> (Output_Get:Material Instance Dynamic Object Reference)
   > SetScalarParameterValue(
    self=Get PipsDynamicMaterial() -> (PipsDynamicMaterial:Material Instance Dynamic Object Reference)
    ParameterName=Options
    Value="To Float (Integer)"(InInt=Get PipCount() -> (PipCount:int=0))
   )
   > SetScalarParameterValue(
    self=Get PipsDynamicMaterial()
    ParameterName=DefaultOption
    Value="To Float (Integer)"(InInt=GetSelectedIndex!(self:Common Rotator Object Reference) -> (ReturnValue:int=0))
   )
  ]}
 OnFocusReceived {nodes [
   OnFocusReceived!() -> (then, MyGeometry:Geometry Structure, InFocusEvent:Focus Event Structure) {is=title K2Node_FunctionEntry}
   > n1 = Branch(Condition=Get OneDirectionalModeActive() -> (OneDirectionalModeActive:bool))
   > Return Node!(execute, ReturnValue:Event Reply Structure=Handled())
   UpdateButtonSizeForMobile(execute=@n1.else)
   > SetRenderTransform!(
    execute
    self:Widget Object Reference=Get ButtonLeft()
    InTransform:Widget Transform Structure=Make Widget Transform!(
     Translation:Vector 2D Structure=(X=0,Y=0)
     Scale:Vector 2D Structure=Get RotatorButtonsScale()
     Shear:Vector 2D Structure=(X=0,Y=0)
     Angle:float=0
    ) -> (WidgetTransform:Widget Transform Structure) {is=title K2Node_MakeStruct}
   ) -> (then)
   n2 = Get ButtonRight()
   > SetVisibility(self=@n2, InVisibility=Visible)
   > SetRenderTransform(
    self=@n2
    InTransform=Make Widget Transform(
     Translation=(X=0,Y=0)
     Scale=GetRightButtonScale!() -> (RightButtonScale:Vector 2D Structure) {is=title K2Node_CallFunction}
     Shear=(X=0,Y=0)
     Angle=0
    )
   )
   n3 = Get CommonInputSubsystem!() -> (ReturnValue:Common Input Object Reference) {is=title K2Node_GetBlueprintContextEx}
   > n4 = Branch(Condition="Equal (Enum)"(A=GetCurrentInputType(self=@n3), B=Gamepad))
   n5 = IsHovered()
   n6 = NOT Boolean(A=Get ShowGamepadButtonOnlyWhenFocused() -> (ShowGamepadButtonOnlyWhenFocused:bool))
   > SetVisibility(InVisibility=Select(Option 0=Hidden, Option 1=Visible, Index=OR Boolean(A=@n6, B=@n5)))
   n7 = Handled()
   > Return Node(ReturnValue=@n7)
   SetVisibility(execute=@n4.else, self=Get ButtonLeft(), InVisibility=Visible)
   > Return Node(ReturnValue=@n7)
  ]}
 UpdateButtonSizeForMobile {nodes [
   UpdateButtonSizeForMobile() {is=title K2Node_FunctionEntry}
   > n1 = Branch(Condition=IsMobileGame())
   > n2 = Set RotatorButtonsScale(RotatorButtonsScale:Vector 2D Structure=(X=1.9,Y=1.9)) -> (Output_Get:Vector 2D Structure)
   Return Node!(execute=[@n3.then, @n2.then])
   n3 = Set RotatorButtonsScale(execute=@n1.else, RotatorButtonsScale=(X=1,Y=1))
  ]}
 UpdatePips {nodes [
   UpdatePips!() -> (then, DoNotShowPips:bool=~) {is=title K2Node_FunctionEntry}
   > SetVisibility(
    self=Get Overlay_Pips()
    InVisibility=Select(Option 0=SelfHitTestInvisible, Option 1=Collapsed, Index=Get DoNotShowPips() -> (DoNotShowPips:bool))
   )
   > n1 = Branch(Condition=Get DoNotShowPips())
   SetScalarParameterValue(
    execute=@n1.else
    self=Get PipsDynamicMaterial()
    ParameterName=SelectedOption
    Value="To Float (Integer)"(InInt=GetSelectedIndex() -> (ReturnValue=0))
   )
  ]}
]
variables [
 BaseOnClickedEnabled {type=boolean, default=true}
 bIsMouseInput {type=boolean}
 Left Triggering Input Action {type=DataTableRowHandle, default {dataTable=&InputActionDataTable, rowName=PlaylistChange}}
 OneDirectionalModeActive {type=boolean}
 PipsDynamicMaterial {type=MaterialInstanceDynamic}
 Right Triggering Input Action {type=DataTableRowHandle, default {dataTable=&InputActionDataTable, rowName=PlaylistChange}}
 RotatorButtonsScale {type=Vector2D, default=<1 1>}
 ShowGamepadButtonOnlyWhenFocused {type=boolean}
]
widgets {
 info {inheritedWidgetCount=2, parentClass=/Script/CommonUI.CommonRotator, rootWidgetClass=&CommonBorder, widgetCount=11}
 widgets [
  MainBorder {
   widgetClassPath=&CommonBorder
   bIsVariable=true
   widget=MainBorder
   properties {
    navigation=&self:WidgetTree.MainBorder.WidgetNavigation_0
    background {
     bIsDynamicallyLoaded=false
     drawAs=NoDrawType
     imageType=NoImage
     mirroring=NoMirror
     resourceName=None
     resourceObject=null
     tiling=NoTile
     imageSize=<32 32>
     margin=<ltrb 0.4 0.4 0.4 0.4>
     outlineSettings {
      bUseBrushTransparency=false
      roundingType=HalfHeightRadius
      width=0
      color {colorUseRule=UseColor_Specified, specifiedColor=<rgba 0 0 0 0>}
      cornerRadii=<0 0 0 0>
     }
     tintColor {colorUseRule=UseColor_Specified, specifiedColor=<rgba 1 1 1 1>}
     uVRegion {bIsValid=false, max=<0 0>, min=<0 0>}
    }
    padding=<ltrb 0 0 0 0>
   }
  }
  HorizontalBox_3 {
   widgetClassPath=&HorizontalBox
   parent=MainBorder
   widget=HorizontalBox_3
   slot {class=BorderSlot, properties {padding=<ltrb 0 0 0 0>}}
  }
  ButtonLeft {
   widgetClassPath=&IconTextButton
   bIsVariable=true
   parent=HorizontalBox_3
   widget=ButtonLeft
   properties {
    bApplyAlphaOnDisable=true
    bHideInputAction=true
    buttonText=""
    onButtonBaseClicked=(&self:WidgetTree.ButtonLeft.HandleOnButtonClicked)
    onButtonBaseDoubleClicked=(&self:WidgetTree.ButtonLeft.HandleOnButtonDoubleClicked)
    onButtonBaseHovered=(&self:WidgetTree.ButtonLeft.HandleOnButtonHovered)
    onButtonBaseUnhovered=(&self:WidgetTree.ButtonLeft.HandleOnButtonUnhovered)
    onSelectedChangedBase=(&self:WidgetTree.ButtonLeft.HandleOnSelectedChanged)
    style=/FortUILibrary/Foundation/Buttons/ButtonStyles/ButtonStyle-Left_CrPropertyEditor.ButtonStyle-Left_CrPropertyEditor_C
   }
   slot {class=HorizontalBoxSlot, properties {horizontalAlignment=HAlign_Left, verticalAlignment=VAlign_Center, padding=<ltrb 8 0 0 0>}}
  }
  VerticalBox_0 {
   widgetClassPath=&VerticalBox
   parent=HorizontalBox_3
   widget=VerticalBox_0
   slot {class=HorizontalBoxSlot, properties {verticalAlignment=VAlign_Center, padding=<ltrb 12 0 12 0>, size {sizeRule=Fill, value=1}}}
  }
  Overlay_1 {widgetClassPath=&Overlay, parent=VerticalBox_0, widget=Overlay_1, slot {class=VerticalBoxSlot}}
  CommonBorder_0 {
   widgetClassPath=&CommonBorder
   parent=Overlay_1
   widget=CommonBorder_0
   properties {
    background {
     bIsDynamicallyLoaded=false
     drawAs=Image
     imageType=NoImage
     mirroring=NoMirror
     resourceName=None
     resourceObject=null
     tiling=NoTile
     imageSize=<32 32>
     margin=<ltrb 0 0 0 0>
     outlineSettings {
      bUseBrushTransparency=false
      roundingType=HalfHeightRadius
      width=0
      color {colorUseRule=UseColor_Specified, specifiedColor=<rgba 0 0 0 0>}
      cornerRadii=<0 0 0 0>
     }
     tintColor {colorUseRule=UseColor_Specified, specifiedColor=<rgba 0.0545 0.4735 0.9647 0.2>}
     uVRegion {bIsValid=false, max=<0 0>, min=<0 0>}
    }
    padding=<ltrb 8 8 8 8>
   }
   slot {class=OverlaySlot, properties {horizontalAlignment=HAlign_Fill, verticalAlignment=VAlign_Fill}}
  }
  MyText {
   widgetClassPath=&CommonTextBlock
   bInherited=true
   bIsVariable=true
   parent=CommonBorder_0
   widget=MyText
   properties {
    justification=Center
    scrollStyle=&TextScrollStyle-Base_C
    text=Option
    colorAndOpacity {colorUseRule=UseColor_Specified, specifiedColor=<rgba 0 0 0 1>}
    font {
     bForceMonospaced=false
     bMaterialIsStencil=false
     fontMaterial=null
     fontObject=&BurbankSmall
     letterSpacing=0
     monospacedWidth=1
     size=18
     skewAmount=0
     typefaceFontName=Bold
     outlineSettings {
      bApplyOutlineToDropShadows=false
      bMiteredCorners=false
      bSeparateFillAlpha=false
      outlineMaterial=null
      outlineSize=0
      outlineColor=<rgba 0 0.0182 0.2031 0.2>
     }
    }
    shadowOffset=<0 0>
   }
   slot {class=BorderSlot, properties {padding=<ltrb 8 8 8 8>}}
  }
  Overlay_Pips {
   widgetClassPath=&Overlay
   bIsVariable=true
   parent=Overlay_1
   widget=Overlay_Pips
   slot {class=OverlaySlot, properties {horizontalAlignment=HAlign_Fill, verticalAlignment=VAlign_Fill}}
  }
  PipIndicator {
   widgetClassPath=&Image
   bIsVariable=true
   parent=Overlay_Pips
   widget=PipIndicator
   properties {
    bFlipForRightToLeftFlowDirection=true
    brush {
     bIsDynamicallyLoaded=false
     drawAs=Image
     imageType=NoImage
     mirroring=NoMirror
     resourceName=None
     resourceObject=/Game/Athena/UI/InGame/Creative/PropertyEditor/Art/M_Cr_RotatorPips_Master_Instance
     tiling=NoTile
     imageSize=<128 32>
     margin=<ltrb 0 0 0 0>
     outlineSettings {
      bUseBrushTransparency=false
      roundingType=HalfHeightRadius
      width=0
      color {colorUseRule=UseColor_Specified, specifiedColor=<rgba 0 0 0 0>}
      cornerRadii=<0 0 0 0>
     }
     tintColor {colorUseRule=UseColor_Specified, specifiedColor=<rgba 1 1 1 1>}
     uVRegion {bIsValid=false, max=<0 0>, min=<0 0>}
    }
   }
   slot {class=OverlaySlot, properties {horizontalAlignment=HAlign_Fill, verticalAlignment=VAlign_Fill, padding=<ltrb 2 0 2 2>}}
  }
  ButtonRight {
   widgetClassPath=&IconTextButton
   bIsVariable=true
   parent=HorizontalBox_3
   widget=ButtonRight
   properties {
    bApplyAlphaOnDisable=true
    bHideInputAction=true
    buttonText=""
    onButtonBaseClicked=(&self:WidgetTree.ButtonRight.HandleOnButtonClicked)
    onButtonBaseDoubleClicked=(&self:WidgetTree.ButtonRight.HandleOnButtonDoubleClicked)
    onButtonBaseHovered=(&self:WidgetTree.ButtonRight.HandleOnButtonHovered)
    onButtonBaseUnhovered=(&self:WidgetTree.ButtonRight.HandleOnButtonUnhovered)
    onSelectedChangedBase=(&self:WidgetTree.ButtonRight.HandleOnSelectedChanged)
    style=/FortUILibrary/Foundation/Buttons/ButtonStyles/ButtonStyle-Right_CrPropertyEditor.ButtonStyle-Right_CrPropertyEditor_C
   }
   slot {class=HorizontalBoxSlot, properties {horizontalAlignment=HAlign_Center, verticalAlignment=VAlign_Center, padding=<ltrb 0 0 8 0>}}
  }
  InputActionWidget {widgetClassPath=&CommonActionWidget, bInherited=true, widget=null}
 ]
}
