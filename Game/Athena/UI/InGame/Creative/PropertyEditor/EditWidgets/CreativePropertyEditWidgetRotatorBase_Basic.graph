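# Creative property editor row widget that lets players pick one of several labeled option values with a rotator beside the property name.
# Init fills the rotator labels, selection and pips from the stepper values and sets the name text.
# Rotating stores the index, marks the entry modified, updates pips and requests selection.
# Modified state is forwarded to the content widget, and focus is centered on and routed to the rotator.
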WidgetBlueprint /Game/Athena/UI/InGame/Creative/PropertyEditor/EditWidgets/CreativePropertyEditWidgetRotatorBase_Basic
graphs [
 BP_GetDesiredFocusTarget {nodes [
   Get Desired Focus Target()
   > Return Node!(
    execute
    ReturnValue:Widget Object Reference=Get CreativePropertyEditWidget_Rotator() -> (CreativePropertyEditWidget_Rotator:Creative Property Edit Widget Rotator Object Reference)
   )
  ]}
 EventGraph {nodes [
   n1 = "On Rotated (CreativePropertyEditWidget_Rotator)"!() -> (OutputDelegate, then, Value:int=0) {is=title K2Node_ComponentBoundEvent}
   > Set CurrentIndexInStepperValue(CurrentIndexInStepperValue:int=@n1.Value) -> (Output_Get:int=0)
   > SetIsModified!(execute, self:Creative Property Edit Widget Base Object Reference, bModified:bool=true) -> (then)
   > UpdatePips!(execute, self:Creative Property Edit Widget Rotator Object Reference=Get CreativePropertyEditWidget_Rotator(), DoNotShowPips:bool=false) -> (then)
   > RequestSelected!(execute, self:Creative Property Edit Widget Base Object Reference, ObjectToSelect:Object Reference, bNavigateToItem:bool=false) -> (then)
   n2 = Event BP_OnModified!() -> (OutputDelegate, then, bModified:bool) {is=title K2Node_Event}
   > UpdateModifiedState!(execute, self, isModifiedFromDefaultValue:bool=@n2.bModified) -> (then)
   n3 = Event SetDesiredValueWidth!() -> (OutputDelegate, then, DesiredValueWidth:float) {is=title K2Node_Event}
   > SetWidthOverride(self=Get SizeBox_Value() -> (SizeBox_Value:Size Box Object Reference), InWidthOverride=@n3.DesiredValueWidth)
   Event BP_PostInitEditWidget() {is=title K2Node_Event}
   > PopulateTextLabels!(
    execute
    self:Common Rotator Object Reference=Get CreativePropertyEditWidget_Rotator()
    Labels:Array of Texts=Get AvailableStepperValueLabels() -> (AvailableStepperValueLabels:Array of Texts)
   ) -> (then)
   > SetSelectedItem!(
    execute
    self:Common Rotator Object Reference=Get CreativePropertyEditWidget_Rotator()
    InValue:int=Get CurrentIndexInStepperValue() -> (CurrentIndexInStepperValue:int=0)
   ) -> (then)
   > UpdateText!(
    execute
    self:WBP Creative Property Editor Content Primary Object Reference=Get WBP_Content_Primary() -> (WBP_Content_Primary:WBP Creative Property Editor Content Primary Object Reference)
    Text:"Text (by ref)"=GetEntryDisplayName!(self:Creative Property Edit Widget Base Object Reference) -> (ReturnValue:Text)
   ) -> (then)
   > InitializePips!(
    execute
    self:Creative Property Edit Widget Rotator Object Reference=Get CreativePropertyEditWidget_Rotator()
    PipCount:int=Length(TargetArray:Array of Texts=Get AvailableStepperValueLabels())
    DoNotShowPips:bool=false
   ) -> (then)
   > UpdatePips(self=Get CreativePropertyEditWidget_Rotator(), DoNotShowPips=false)
  ]}
 OnFocusReceived {nodes [
   OnFocusReceived!() -> (then, MyGeometry:Geometry Structure, InFocusEvent:Focus Event Structure) {is=title K2Node_FunctionEntry}
   n1 = Get CreativePropertyEditWidget_Rotator()
   > CenterWidget!(execute, ToCenter:Widget Object Reference=@n1) -> (then)
   > SetKeyboardFocus(self:Widget Object Reference=@n1)
   > Return Node(ReturnValue:Event Reply Structure=Handled())
  ]}
 UpdateModifiedState {nodes [
   n1 = UpdateModifiedState!() -> (then, isModifiedFromDefaultValue:bool=~) {is=title K2Node_FunctionEntry}
   > Event Update IsModified!(execute, self:WBP Creative Property Editor Content Primary Object Reference=Get WBP_Content_Primary(), bIsModified:bool=@n1) -> (then)
  ]}
]
variables []
widgets {
 info {
  inheritedWidgetCount=1
  parentClass=/Script/FortniteUI.CreativePropertyEditWidgetUserOptionBase
  rootWidgetClass=&Overlay
  widgetCount=12
 }
 widgets [
  Overlay_25 {widgetClassPath=&Overlay, widget=Overlay_25}
  SizeBox_2 {
   widgetClassPath=&SizeBox
   parent=Overlay_25
   widget=SizeBox_2
   properties {bOverride_HeightOverride=true, heightOverride=82}
   slot {class=OverlaySlot, properties {horizontalAlignment=HAlign_Fill, verticalAlignment=VAlign_Center, padding=<ltrb 16 0 16 0>}}
  }
  Overlay_0 {widgetClassPath=&Overlay, parent=SizeBox_2, widget=Overlay_0, slot {class=SizeBoxSlot}}
  WBP_Background_SkewedWithMarker {
   widgetClassPath=/Game/Athena/UI/InGame/Creative/PropertyEditor/WBP_CreativePropertyEditor_Background_SkewedWithMarker
   bIsVariable=true
   parent=Overlay_0
   widget=WBP_Background_SkewedWithMarker
   properties {focusState=Hovered}
   slot {class=OverlaySlot, properties {horizontalAlignment=HAlign_Fill, verticalAlignment=VAlign_Fill}}
  }
  HorizontalBox_2 {
   widgetClassPath=&HorizontalBox
   parent=Overlay_0
   widget=HorizontalBox_2
   slot {class=OverlaySlot, properties {horizontalAlignment=HAlign_Fill, verticalAlignment=VAlign_Fill}}
  }
  WBP_Content_Primary {
   widgetClassPath=/Game/Athena/UI/InGame/Creative/PropertyEditor/WBP_CreativePropertyEditor_Content_Primary
   bIsVariable=true
   parent=HorizontalBox_2
   widget=WBP_Content_Primary
   properties {richText="[Parameter Name]", states=DefaultHovered}
   slot {class=HorizontalBoxSlot, properties {verticalAlignment=VAlign_Center, padding=<ltrb 30 0 4 0>, size {sizeRule=Fill, value=1}}}
  }
  SizeBox_Value {
   widgetClassPath=&SizeBox
   bIsVariable=true
   parent=HorizontalBox_2
   widget=SizeBox_Value
   properties {bOverride_WidthOverride=true, widthOverride=320}
   slot {class=HorizontalBoxSlot, properties {verticalAlignment=VAlign_Center}}
  }
  Overlay_2 {widgetClassPath=&Overlay, parent=SizeBox_Value, widget=Overlay_2, slot {class=SizeBoxSlot}}
  OptionSelectionArea_2 {
   widgetClassPath=&HorizontalBox
   bIsVariable=true
   parent=Overlay_2
   widget=OptionSelectionArea_2
   slot {class=OverlaySlot, properties {horizontalAlignment=HAlign_Fill}}
  }
  CreativePropertyEditWidget_Rotator {
   widgetClassPath=/Game/Athena/UI/InGame/Creative/PropertyEditor/Property_Picker/CreativePropertyEditWidget_Rotator
   bIsVariable=true
   parent=OptionSelectionArea_2
   widget=CreativePropertyEditWidget_Rotator
   properties {foregroundColor {colorUseRule=UseColor_Foreground, specifiedColor=<rgba 1 0 1 1>}}
   slot {class=HorizontalBoxSlot, properties {size {sizeRule=Fill, value=1}}}
  }
  Button_Collapse {
   widgetClassPath=&IconTextButton
   bInherited=true
   parent=Overlay_25
   widget=Button_Collapse
   properties {
    onButtonBaseClicked=(&self:WidgetTree.Button_Collapse.HandleOnButtonClicked)
    onButtonBaseDoubleClicked=(&self:WidgetTree.Button_Collapse.HandleOnButtonDoubleClicked)
    onButtonBaseHovered=(&self:WidgetTree.Button_Collapse.HandleOnButtonHovered)
    onButtonBaseUnhovered=(&self:WidgetTree.Button_Collapse.HandleOnButtonUnhovered)
    onSelectedChangedBase=(&self:WidgetTree.Button_Collapse.HandleOnSelectedChanged)
    visibility=Collapsed
   }
   slot {class=OverlaySlot, properties {horizontalAlignment=HAlign_Center, verticalAlignment=VAlign_Center}}
  }
  SaveSpinner_2 {
   widgetClassPath=&Image
   bIsVariable=true
   parent=Overlay_25
   widget=SaveSpinner_2
   properties {
    visibility=Collapsed
    brush {
     bIsDynamicallyLoaded=false
     drawAs=Image
     imageType=NoImage
     mirroring=NoMirror
     resourceName=None
     resourceObject=/FortUILibrary/UIKit/Throbber/MI_UIKit_Color_Throbber
     tiling=NoTile
     imageSize=<24 24>
     margin=<ltrb 0 0 0 0>
     outlineSettings {
      bUseBrushTransparency=false
      roundingType=HalfHeightRadius
      width=0
      color {colorUseRule=UseColor_Specified, specifiedColor=<rgba 0 0 0 0>}
      cornerRadii=<0 0 0 0>
     }
     tintColor {colorUseRule=UseColor_Specified, specifiedColor=<rgba 1 1 1 1>}
     uVRegion {bIsValid=false, max=<0 0>, min=<0 0>}
    }
   }
   slot {class=OverlaySlot, properties {verticalAlignment=VAlign_Center, padding=<ltrb 10 -24 0 0>}}
  }
 ]
}
