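# Creative property editor row widget that edits a numeric device option through a preset rotator combined with a direct numeric entry field.
# Rotator changes update the stepper index, pips, display text and modified state.
# The numeric picker and its progress bar show only for numeric, in-range values, treating a -1 minimum as a non-numeric sentinel with zero as the lower bound.
# Input-type checks steer centering and focus between rotator and entry.
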WidgetBlueprint /Game/Athena/UI/InGame/Creative/PropertyEditor/EditWidgets/CreativePropertyEditWidget_DirectInput_Numeric_Basic
graphs [
 BP_GetDesiredFocusTarget {nodes [
   Get Desired Focus Target()
   > Return Node!(
    execute
    ReturnValue:Widget Object Reference=Get CreativePropertyEditWidget_Rotator() -> (CreativePropertyEditWidget_Rotator:Creative Property Edit Widget Rotator Object Reference)
   )
  ]}
 EventGraph {nodes [
   n1 = "On Value Changed (CreativeProperty_NumericEntry)"!() -> (OutputDelegate, then, InValue:float) {is=title K2Node_ComponentBoundEvent}
   > UpdateNumericValues!(execute, self, InValue:double=@n1.InValue) -> (then)
   > Branch(Condition=NOT Boolean(A=IsUsingKeyboardAndMouse!() -> (ReturnValue:bool)))
   > Branch(Condition=IsNumericPickerEnabled!() -> (ReturnValue:bool))
   > CenterWidget!(execute, ToCenter:Widget Object Reference=Get CreativePropertyEditWidget_Rotator()) -> (then)
   > RequestCloseDialog!(execute, self:Creative Property Edit Widget Base Object Reference, bSetFocusToOptionsList:bool=true) -> (then)
   "On Button Clicked (FocusButton)"!() -> (OutputDelegate, then, Button:Common Button Legacy Object Reference) {is=title K2Node_ComponentBoundEvent}
   > Branch(Condition=NOT Boolean(A=IsUsingKeyboardAndMouse()))
   > Branch(Condition=IsNumericPickerEnabled())
   n2 = Get CreativeProperty_NumericEntry() -> (CreativeProperty_NumericEntry:Fort Creative Property Numeric Entry Object Reference)
   > CenterWidget(ToCenter=@n2)
   > SetKeyboardFocus(self:Widget Object Reference=@n2)
   n3 = "On Rotated (CreativePropertyEditWidget_Rotator)"!() -> (OutputDelegate, then, Value:int=0) {is=title K2Node_ComponentBoundEvent}
   > Set CurrentIndexInStepperValue(CurrentIndexInStepperValue:int=@n3.Value) -> (Output_Get:int=0)
   > OnRotated(self:Creative Property Edit Widget Direct Numeric Input Base Object Reference)
   > CheckIsNumeircValueModified(self:Creative Property Edit Widget Direct Numeric Input Base Object Reference)
   > UpdatePips!(execute, self:Creative Property Edit Widget Rotator Object Reference=Get CreativePropertyEditWidget_Rotator(), DoNotShowPips:bool=false) -> (then)
   > RequestSelected!(execute, self:Creative Property Edit Widget Base Object Reference, ObjectToSelect:Object Reference, bNavigateToItem:bool=false) -> (then)
   > UpdateNumericValues_VisualOnly()
   > Branch(Condition=OR Boolean(A=IsUsingKeyboardAndMouse(), B=IsMobileGame()))
   > Branch(Condition=IsNumericPickerEnabled())
   n4 = Get CreativeProperty_NumericEntry()
   > CenterWidget(ToCenter=@n4)
   > SetKeyboardFocus(self=@n4)
   n5 = Get CreativePropertyEditWidget_Rotator()
   > CenterWidget(ToCenter=@n5)
   > SetKeyboardFocus(self=@n5)
   n6 = Event BP_OnModified!() -> (OutputDelegate, then, bModified:bool) {is=title K2Node_Event}
   > UpdateModifiedState!(execute, self, isModifiedFromDefaultValue:bool=@n6.bModified) -> (then)
   > UpdateDisplayText()
   > SetNumericPickerVisibility()
   Event OnInitialized()
   n7 = Get CreativePropertyEditWidget_Rotator()
   > SetIsFocusable!(
    execute
    self:Common Button Base Object Reference=[
     Get ButtonLeft!(self:Creative Property Edit Widget Rotator Object Reference=@n7) -> (ButtonLeft:Icon Text Button Object Reference)
     Get ButtonRight!(self:Creative Property Edit Widget Rotator Object Reference=@n7) -> (ButtonRight:Icon Text Button Object Reference)
    ]
    bInIsFocusable:bool=false
   ) -> (then)
   n8 = Event BP_PostNumericUpdateEditWidget!() -> (OutputDelegate, then, IsCustomValue:bool) {is=title K2Node_Event}
   > n9 = Branch(Condition=@n8.IsCustomValue)
   SetSelectedItem!(
    execute=@n9.else
    self:Common Rotator Object Reference=Get CreativePropertyEditWidget_Rotator()
    InValue:int=Get CurrentIndexInStepperValue() -> (CurrentIndexInStepperValue:int=0)
   ) -> (then)
   > UpdatePips(self=Get CreativePropertyEditWidget_Rotator(), DoNotShowPips=false)
   > UpdateNumericValues_VisualOnly()
   > UpdateDisplayText()
   n10 = Event SetDesiredValueWidth!() -> (OutputDelegate, then, DesiredValueWidth:float) {is=title K2Node_Event}
   > SetWidthOverride(self=Get SizeBox_Value() -> (SizeBox_Value:Size Box Object Reference), InWidthOverride=@n10.DesiredValueWidth)
   Event BP_PostInitEditWidget() {is=title K2Node_Event}
   > PopulateTextLabels!(
    execute
    self:Common Rotator Object Reference=Get CreativePropertyEditWidget_Rotator()
    Labels:Array of Texts=Get AvailableStepperValueLabels() -> (AvailableStepperValueLabels:Array of Texts)
   ) -> (then)
   > SetSelectedItem(self=Get CreativePropertyEditWidget_Rotator(), InValue=Get CurrentIndexInStepperValue() -> (CurrentIndexInStepperValue=0))
   > UpdateText!(
    execute
    self:WBP Creative Property Editor Content Primary Object Reference=Get WBP_Content_Primary() -> (WBP_Content_Primary:WBP Creative Property Editor Content Primary Object Reference)
    Text:"Text (by ref)"=GetEntryDisplayName!(self:Creative Property Edit Widget Base Object Reference) -> (ReturnValue:Text)
   ) -> (then)
   > InitializePips!(
    execute
    self:Creative Property Edit Widget Rotator Object Reference=Get CreativePropertyEditWidget_Rotator()
    PipCount:int=Length(TargetArray:Array of Texts=Get AvailableStepperValueLabels())
    DoNotShowPips:bool=false
   ) -> (then)
   > UpdatePips(self=Get CreativePropertyEditWidget_Rotator(), DoNotShowPips=false)
   > UpdateDisplayText()
   > InitializeNumericPicker()
   Event OnMouseLeave!() -> (OutputDelegate, then, MouseEvent:"Pointer Event Structure (by ref)") {is=title K2Node_Event}
   > Branch(Condition=IsUsingKeyboardAndMouse())
   > CenterWidget(ToCenter=Get CreativePropertyEditWidget_Rotator())
   > SetKeyboardFocus(self=Get CreativePropertyEditWidget_Rotator())
   > CheckIsNumeircValueModified()
   > UpdateDisplayText()
  ]}
 InitializeNumericPicker {nodes [
   InitializeNumericPicker() {is=title K2Node_FunctionEntry}
   > SetNumericPickerVisibility()
   > n1 = Set IsStartWithIllegalValue(IsStartWithIllegalValue:bool="Equal (Float)"(A=Get DeviceOptionValueMin() -> (DeviceOptionValueMin:float), B=-1)) -> (Output_Get:bool)
   > Set DeviceOptionMinValueInternal(DeviceOptionMinValueInternal:double=SelectFloat(A=0, B=Get DeviceOptionValueMin(), bPickA=@n1)) -> (Output_Get:double)
   > Set DeviceOptionMaxValueInternal(DeviceOptionMaxValueInternal:double=Get DeviceOptionValueMax() -> (DeviceOptionValueMax:float)) -> (Output_Get:double)
   > UpdateNumericValues_VisualOnly()
   > UpdateNumericValues_UI!(execute, self, InValue:double=GetValue!(self:Spin Box Object Reference=Get CreativeProperty_NumericEntry()) -> (ReturnValue:float)) -> (then)
  ]}
 IsNumericPickerEnabled {nodes [
   IsNumericPickerEnabled() {is=title K2Node_FunctionEntry}
   n1 = GetValue(self=Get CreativeProperty_NumericEntry())
   n2 = Get DeviceOptionMinValueInternal() -> (DeviceOptionMinValueInternal:double)
   n3 = Get DeviceOptionMaxValueInternal() -> (DeviceOptionMaxValueInternal:double)
   > Return Node(ReturnValue:bool=AND Boolean(
     A=Get IsCurrentValueNumeric() -> (IsCurrentValueNumeric:bool)
     B="In Range (Float)"!(Value:double=@n1, Min:double=@n2, Max:double=@n3, InclusiveMin:bool=true, InclusiveMax:bool=true) -> (ReturnValue:bool)
    ))
  ]}
 IsNumericPickerExcludeIllegalValues {nodes [
   n1 = IsNumericPickerExcludeIllegalValues!() -> (then, InString:String) {is=title K2Node_FunctionEntry}
   n2 = Get DeviceOptionMinValueInternal()
   n3 = Get DeviceOptionMaxValueInternal()
   n4 = String To Float(InString=@n1)
   n5 = AND Boolean(A=Get IsStartWithIllegalValue() -> (IsStartWithIllegalValue:bool), B="Equal (Float)"(A=@n4, B:int=-1))
   > Return Node(ReturnValue=AND Boolean!(
     A=IsNumeric!(SourceString:String=@n1) -> (ReturnValue:bool)
     B=NOT Boolean(A=@n5)
     C:bool="In Range (Float)"(Value=@n4, Min=@n2, Max=@n3, InclusiveMin=true, InclusiveMax=true)
    ) -> (ReturnValue))
  ]}
 IsUsingKeyboardAndMouse {nodes [
   IsUsingKeyboardAndMouse() {is=title K2Node_FunctionEntry}
   n1 = Get CommonInputSubsystem()
   > Return Node(ReturnValue="Equal (Enum)"(A=GetCurrentInputType(self=@n1), B=MouseAndKeyboard))
  ]}
 OnFocusReceived {nodes [
   OnFocusReceived!() -> (then, MyGeometry:Geometry Structure, InFocusEvent:Focus Event Structure) {is=title K2Node_FunctionEntry}
   > n1 = Branch(Condition=IsNumericPickerEnabled() {is=title K2Node_CallFunction})
   > n2 = Branch(Condition=IsUsingKeyboardAndMouse() {is=title K2Node_CallFunction})
   > CenterWidget(ToCenter=Get CreativePropertyEditWidget_Rotator())
   > n3 = CenterWidget(ToCenter=Get CreativeProperty_NumericEntry())
   Return Node(execute=[@n3.then, @n4.then], ReturnValue:Event Reply Structure=Handled())
   n4 = CenterWidget(execute=[@n2.else, @n1.else], ToCenter=Get CreativePropertyEditWidget_Rotator())
  ]}
 SetNumericPickerVisibility {nodes [
   SetNumericPickerVisibility() {is=title K2Node_FunctionEntry}
   > SetVisibility(
    self=Get Overlay_NumericPicker() -> (Overlay_NumericPicker:Overlay Object Reference)
    InVisibility=Select(Option 0=Hidden, Option 1=Visible, Index=IsNumericPickerEnabled())
   )
  ]}
 UpdateDisplayText {nodes [
   UpdateDisplayText() {is=title K2Node_FunctionEntry}
   > "SetText (Text)"(
    self=Get MyText!(self:Common Rotator Object Reference=Get CreativePropertyEditWidget_Rotator()) -> (MyText:Common Text Object Reference)
    InText=GetEntryDisplayValue!(self:Creative Property Edit Widget Base Object Reference) -> (ReturnValue:Text)
   )
  ]}
 UpdateModifiedState {nodes [
   n1 = UpdateModifiedState!() -> (then, isModifiedFromDefaultValue:bool=~) {is=title K2Node_FunctionEntry}
   > Event Update IsModified!(execute, self:WBP Creative Property Editor Content Primary Object Reference=Get WBP_Content_Primary(), bIsModified:bool=@n1) -> (then)
  ]}
 UpdateNumericValues {nodes [
   n1 = UpdateNumericValues!() -> (then, InValue:double=~) {is=title K2Node_FunctionEntry}
   > UpdateNumericValues_UI(InValue=@n1)
   > n2 = Branch(Condition=Get IsVisualUpdateOnly() -> (IsVisualUpdateOnly:bool))
   > Set IsVisualUpdateOnly(IsVisualUpdateOnly:bool=false) -> (Output_Get:bool)
   UpdateNumericValues_VisualOnly(execute=@n2.else)
   > CheckIsNumeircValueModified()
  ]}
 UpdateNumericValues_UI {nodes [
   n1 = UpdateNumericValues_UI!() -> (then, InValue:double=~) {is=title K2Node_FunctionEntry}
   > Branch(Condition="float > float"(A=Get DeviceOptionMaxValueInternal()))
   n2 = float / float(
    A=float - float(A=@n1, B=Get DeviceOptionMinValueInternal())
    B=float - float(A=Get DeviceOptionMaxValueInternal(), B=Get DeviceOptionMinValueInternal())
   )
   > SetProgress!(
    execute
    self:Fort Simple Material Progress Bar Object Reference=Get Progress() -> (Progress:Simple Material Progress Bar Object Reference)
    Progress:float=@n2
   ) -> (then)
   > n3 = GetDynamicMaterial(self=Get DefaultValueLayer() -> (DefaultValueLayer:Image Object Reference))
   > SetScalarParameterValue(self=@n3, ParameterName=DefaultValue, Value=@n2)
  ]}
 UpdateNumericValues_VisualOnly {nodes [
   UpdateNumericValues_VisualOnly() {is=title K2Node_FunctionEntry}
   > Set IsVisualUpdateOnly(IsVisualUpdateOnly=true)
   > n1 = ParseTeamOrClassOptionsAsString!(
    execute
    self:Creative Property Edit Widget Direct Numeric Input Base Object Reference
    OriginalString:String=TryGetCurrentOptionValue!(self:Creative Property Edit Widget Direct Numeric Input Base Object Reference) -> (Data:String)
   ) -> (then, ReturnValue:String)
   > n2 = Set IsCurrentValueNumeric(IsCurrentValueNumeric:bool=IsNumericPickerExcludeIllegalValues!(InString:String=@n1) -> (ReturnValue:bool) {is=title K2Node_CallFunction}) -> (Output_Get:bool)
   > SetNumericEntryEditable!(execute, self:Creative Property Edit Widget Direct Numeric Input Base Object Reference, bIsEditable:bool=@n2) -> (then)
   > SetNumericPickerVisibility() {is=title K2Node_CallFunction}
  ]}
]
variables [
 DeviceOptionMaxValueInternal {type=number}
 DeviceOptionMinValueInternal {type=number}
 IsCurrentValueNumeric {type=boolean}
 IsStartWithIllegalValue {type=boolean}
 IsVisualUpdateOnly {type=boolean}
]
widgets {
 info {
  inheritedWidgetCount=2
  parentClass=/Script/FortniteUI.CreativePropertyEditWidgetDirectNumericInputBase
  rootWidgetClass=&Overlay
  widgetCount=20
 }
 widgets [
  Overlay_25 {widgetClassPath=&Overlay, widget=Overlay_25}
  SizeBox_2 {
   widgetClassPath=&SizeBox
   parent=Overlay_25
   widget=SizeBox_2
   properties {bOverride_HeightOverride=true, heightOverride=82}
   slot {class=OverlaySlot, properties {horizontalAlignment=HAlign_Fill, verticalAlignment=VAlign_Center, padding=<ltrb 16 0 16 0>}}
  }
  Overlay_0 {widgetClassPath=&Overlay, parent=SizeBox_2, widget=Overlay_0, slot {class=SizeBoxSlot}}
  WBP_Background_SkewedWithMarker {
   widgetClassPath=/Game/Athena/UI/InGame/Creative/PropertyEditor/WBP_CreativePropertyEditor_Background_SkewedWithMarker
   bIsVariable=true
   parent=Overlay_0
   widget=WBP_Background_SkewedWithMarker
   properties {focusState=Hovered}
   slot {class=OverlaySlot, properties {horizontalAlignment=HAlign_Fill, verticalAlignment=VAlign_Fill}}
  }
  HorizontalBox_2 {
   widgetClassPath=&HorizontalBox
   parent=Overlay_0
   widget=HorizontalBox_2
   slot {class=OverlaySlot, properties {horizontalAlignment=HAlign_Fill, verticalAlignment=VAlign_Fill}}
  }
  WBP_Content_Primary {
   widgetClassPath=/Game/Athena/UI/InGame/Creative/PropertyEditor/WBP_CreativePropertyEditor_Content_Primary
   bIsVariable=true
   parent=HorizontalBox_2
   widget=WBP_Content_Primary
   properties {richText="[Parameter Name]", states=DefaultHovered}
   slot {class=HorizontalBoxSlot, properties {verticalAlignment=VAlign_Center, padding=<ltrb 30 0 4 0>, size {sizeRule=Fill, value=1}}}
  }
  SizeBox_Value {
   widgetClassPath=&SizeBox
   bIsVariable=true
   parent=HorizontalBox_2
   widget=SizeBox_Value
   properties {bOverride_WidthOverride=true, widthOverride=320}
   slot {class=HorizontalBoxSlot, properties {verticalAlignment=VAlign_Center}}
  }
  Overlay_2 {widgetClassPath=&Overlay, parent=SizeBox_Value, widget=Overlay_2, slot {class=SizeBoxSlot}}
  OptionSelectionArea_2 {
   widgetClassPath=&HorizontalBox
   bIsVariable=true
   parent=Overlay_2
   widget=OptionSelectionArea_2
   slot {class=OverlaySlot, properties {horizontalAlignment=HAlign_Fill}}
  }
  CreativePropertyEditWidget_Rotator {
   widgetClassPath=/Game/Athena/UI/InGame/Creative/PropertyEditor/Property_Picker/CreativePropertyEditWidget_Rotator
   bIsVariable=true
   parent=OptionSelectionArea_2
   widget=CreativePropertyEditWidget_Rotator
   properties {foregroundColor {colorUseRule=UseColor_Foreground, specifiedColor=<rgba 1 0 1 1>}}
   slot {class=HorizontalBoxSlot, properties {size {sizeRule=Fill, value=1}}}
  }
  SizeBox_63 {
   widgetClassPath=&SizeBox
   parent=Overlay_2
   widget=SizeBox_63
   properties {bOverride_HeightOverride=true, bOverride_WidthOverride=true, heightOverride=34, widthOverride=210}
   slot {class=OverlaySlot, properties {horizontalAlignment=HAlign_Center}}
  }
  Overlay_NumericPicker {widgetClassPath=&Overlay, bIsVariable=true, parent=SizeBox_63, widget=Overlay_NumericPicker, slot {class=SizeBoxSlot}}
  FocusButton {
   widgetClassPath=&IconTextButton
   bIsVariable=true
   parent=Overlay_NumericPicker
   widget=FocusButton
   properties {
    contentAlignment=HAlign_Center
    minWidth=256
    onButtonBaseClicked=(&self:WidgetTree.FocusButton.HandleOnButtonClicked)
    onButtonBaseDoubleClicked=(&self:WidgetTree.FocusButton.HandleOnButtonDoubleClicked)
    onButtonBaseHovered=(&self:WidgetTree.FocusButton.HandleOnButtonHovered)
    onButtonBaseUnhovered=(&self:WidgetTree.FocusButton.HandleOnButtonUnhovered)
    onSelectedChangedBase=(&self:WidgetTree.FocusButton.HandleOnSelectedChanged)
    style=null
    hoveredSlateSoundOverride {resourceObject=null}
    pressedSlateSoundOverride {resourceObject=null}
    triggeringInputAction {dataTable=&InputActionDataTable, rowName=GenericForward}
   }
   slot {class=OverlaySlot, properties {horizontalAlignment=HAlign_Fill, verticalAlignment=VAlign_Fill}}
  }
  Progress {
   widgetClassPath=/Game/Athena/HUD/SimpleMaterialProgressBar
   bIsVariable=true
   parent=Overlay_NumericPicker
   widget=Progress
   properties {
    bOverrideDefaultColorA=true
    bOverrideDefaultColorB=true
    bOverrideDefaultColorBackground=true
    glowEdge=-0.25
    segmentEdge=0.0002
    segments=10
    visibility=HitTestInvisible
    cachedColorA=<rgba 0.0382 0.4452 0.9047 1>
    cachedColorB=<rgba 0.0382 0.4452 0.9047 1>
    cachedColorBackground=<rgba 0.0033 0.0097 0.1022 1>
   }
   slot {class=OverlaySlot, properties {horizontalAlignment=HAlign_Fill, verticalAlignment=VAlign_Fill}}
  }
  DefaultValueLayer {
   widgetClassPath=&Image
   bIsVariable=true
   parent=Overlay_NumericPicker
   widget=DefaultValueLayer
   properties {
    visibility=HitTestInvisible
    brush {
     bIsDynamicallyLoaded=false
     drawAs=Image
     imageType=NoImage
     mirroring=NoMirror
     resourceName=None
     resourceObject=/Game/Athena/Settings/Art/SettingsDefaultSlider_M
     tiling=NoTile
     imageSize=<32 32>
     margin=<ltrb 0 0 0 0>
     outlineSettings {
      bUseBrushTransparency=false
      roundingType=HalfHeightRadius
      width=0
      color {colorUseRule=UseColor_Specified, specifiedColor=<rgba 0 0 0 0>}
      cornerRadii=<0 0 0 0>
     }
     tintColor {colorUseRule=UseColor_Specified, specifiedColor=<rgba 1 1 1 1>}
     uVRegion {bIsValid=false, max=<0 0>, min=<0 0>}
    }
   }
   slot {class=OverlaySlot, properties {horizontalAlignment=HAlign_Fill, verticalAlignment=VAlign_Fill}}
  }
  Overlay_NumericEntry {
   widgetClassPath=&Overlay
   parent=Overlay_NumericPicker
   widget=Overlay_NumericEntry
   slot {class=OverlaySlot, properties {horizontalAlignment=HAlign_Fill, verticalAlignment=VAlign_Fill}}
  }
  Switcher_NumericEntry {
   widgetClassPath=&CommonActivatableWidgetSwitcher
   bIsVariable=true
   parent=Overlay_NumericEntry
   widget=Switcher_NumericEntry
   slot {class=OverlaySlot, properties {horizontalAlignment=HAlign_Fill, verticalAlignment=VAlign_Fill}}
  }
  CreativeProperty_NumericEntry {
   widgetClassPath=/Script/FortniteUI.FortCreativePropertyNumericEntry
   bInherited=true
   bIsVariable=true
   parent=Switcher_NumericEntry
   widget=CreativeProperty_NumericEntry
   properties {
    bEnableSlider=false
    clearKeyboardFocusOnCommit=true
    justification=Center
    keyboardType=AlphaNumeric
    selectAllTextOnCommit=false
    font {
     bForceMonospaced=false
     bMaterialIsStencil=false
     fontMaterial=null
     fontObject=&BurbankSmall
     letterSpacing=0
     monospacedWidth=1
     size=18
     skewAmount=0
     typefaceFontName=Bold
     outlineSettings {
      bApplyOutlineToDropShadows=false
      bMiteredCorners=false
      bSeparateFillAlpha=false
      outlineMaterial=null
      outlineSize=0
      outlineColor=<rgba 0 0.0182 0.2031 0.2>
     }
    }
    foregroundColor {colorUseRule=UseColor_Specified, specifiedColor=<rgba 1 1 1 1>}
    widgetStyle {
     activeBackgroundBrush {
      bIsDynamicallyLoaded=false
      drawAs=Image
      imageType=NoImage
      mirroring=NoMirror
      resourceName=None
      resourceObject=/Game/Global/Textures/Default/Blanks/T_Blank_White_ZeroAlpha_sRGBColor
      tiling=NoTile
      imageSize=<1 1>
      margin=<ltrb 0 0 0 0>
      outlineSettings {
       bUseBrushTransparency=false
       roundingType=HalfHeightRadius
       width=0
       color {colorUseRule=UseColor_Specified, specifiedColor=<rgba 0 0 0 0>}
       cornerRadii=<0 0 0 0>
      }
      tintColor {colorUseRule=UseColor_Specified, specifiedColor=<rgba 1 1 1 1>}
      uVRegion {bIsValid=false, max=<0 0>, min=<0 0>}
     }
     activeFillBrush {
      bIsDynamicallyLoaded=false
      drawAs=Box
      imageType=FullColor
      mirroring=NoMirror
      resourceName=../../../Engine/Content/Slate/Common/Spinbox_Fill.png
      resourceObject=&T_Blank_White_ZeroAlpha_sRGBColor
      tiling=NoTile
      imageSize=<1 1>
      margin=<ltrb 0.25 0.25 0.5 0.25>
      outlineSettings {
       bUseBrushTransparency=false
       roundingType=HalfHeightRadius
       width=0
       color {colorUseRule=UseColor_Specified, specifiedColor=<rgba 0 0 0 0>}
       cornerRadii=<0 0 0 0>
      }
      tintColor {colorUseRule=UseColor_Specified, specifiedColor=<rgba 1 1 1 1>}
      uVRegion {bIsValid=false, max=<0 0>, min=<0 0>}
     }
     arrowsImage {
      bIsDynamicallyLoaded=false
      drawAs=Image
      imageType=FullColor
      mirroring=NoMirror
      resourceName=../../../Engine/Content/Slate/Common/SpinArrows.png
      resourceObject=&T_Blank_White_ZeroAlpha_sRGBColor
      tiling=NoTile
      imageSize=<1 1>
      margin=<ltrb 0 0 0 0>
      outlineSettings {
       bUseBrushTransparency=false
       roundingType=HalfHeightRadius
       width=0
       color {colorUseRule=UseColor_Specified, specifiedColor=<rgba 0 0 0 0>}
       cornerRadii=<0 0 0 0>
      }
      tintColor {colorUseRule=UseColor_Specified, specifiedColor=<rgba 1 1 1 1>}
      uVRegion {bIsValid=false, max=<0 0>, min=<0 0>}
     }
     backgroundBrush {
      bIsDynamicallyLoaded=false
      drawAs=Box
      imageType=FullColor
      mirroring=NoMirror
      resourceName=../../../Engine/Content/Slate/Common/Spinbox.png
      resourceObject=&T_Blank_White_ZeroAlpha_sRGBColor
      tiling=NoTile
      imageSize=<1 1>
      margin=<ltrb 0.25 0.25 0.25 0.25>
      outlineSettings {
       bUseBrushTransparency=false
       roundingType=HalfHeightRadius
       width=0
       color {colorUseRule=UseColor_Specified, specifiedColor=<rgba 0 0 0 0>}
       cornerRadii=<0 0 0 0>
      }
      tintColor {colorUseRule=UseColor_Specified, specifiedColor=<rgba 1 1 1 1>}
      uVRegion {bIsValid=false, max=<0 0>, min=<0 0>}
     }
     foregroundColor {colorUseRule=UseColor_Specified, specifiedColor=<rgba 0.4729 0.4729 0.4729 1>}
     hoveredBackgroundBrush {
      bIsDynamicallyLoaded=false
      drawAs=Box
      imageType=FullColor
      mirroring=NoMirror
      resourceName=../../../Engine/Content/Slate/Common/Spinbox_Hovered.png
      resourceObject=&T_Blank_White_ZeroAlpha_sRGBColor
      tiling=NoTile
      imageSize=<1 1>
      margin=<ltrb 0.25 0.25 0.25 0.25>
      outlineSettings {
       bUseBrushTransparency=false
       roundingType=HalfHeightRadius
       width=0
       color {colorUseRule=UseColor_Specified, specifiedColor=<rgba 0 0 0 0>}
       cornerRadii=<0 0 0 0>
      }
      tintColor {colorUseRule=UseColor_Specified, specifiedColor=<rgba 1 1 1 1>}
      uVRegion {bIsValid=false, max=<0 0>, min=<0 0>}
     }
     hoveredFillBrush {
      bIsDynamicallyLoaded=false
      drawAs=Image
      imageType=NoImage
      mirroring=NoMirror
      resourceName=None
      resourceObject=&T_Blank_White_ZeroAlpha_sRGBColor
      tiling=NoTile
      imageSize=<1 1>
      margin=<ltrb 0 0 0 0>
      outlineSettings {
       bUseBrushTransparency=false
       roundingType=HalfHeightRadius
       width=0
       color {colorUseRule=UseColor_Specified, specifiedColor=<rgba 0 0 0 0>}
       cornerRadii=<0 0 0 0>
      }
      tintColor {colorUseRule=UseColor_Specified, specifiedColor=<rgba 1 1 1 1>}
      uVRegion {bIsValid=false, max=<0 0>, min=<0 0>}
     }
     inactiveFillBrush {
      bIsDynamicallyLoaded=false
      drawAs=Box
      imageType=FullColor
      mirroring=NoMirror
      resourceName=../../../Engine/Content/Slate/Common/Spinbox_Fill_Hovered.png
      resourceObject=&T_Blank_White_ZeroAlpha_sRGBColor
      tiling=NoTile
      imageSize=<1 1>
      margin=<ltrb 0.25 0.25 0.25 0.25>
      outlineSettings {
       bUseBrushTransparency=false
       roundingType=HalfHeightRadius
       width=0
       color {colorUseRule=UseColor_Specified, specifiedColor=<rgba 0 0 0 0>}
       cornerRadii=<0 0 0 0>
      }
      tintColor {colorUseRule=UseColor_Specified, specifiedColor=<rgba 1 1 1 1>}
      uVRegion {bIsValid=false, max=<0 0>, min=<0 0>}
     }
     insetPadding=<ltrb 0 0 0 0>
     textPadding=<ltrb 1 2 1 2>
    }
   }
   slot {class=WidgetSwitcherSlot}
  }
  Button_Collapse {
   widgetClassPath=&IconTextButton
   bInherited=true
   parent=Overlay_25
   widget=Button_Collapse
   properties {
    onButtonBaseClicked=(&self:WidgetTree.Button_Collapse.HandleOnButtonClicked)
    onButtonBaseDoubleClicked=(&self:WidgetTree.Button_Collapse.HandleOnButtonDoubleClicked)
    onButtonBaseHovered=(&self:WidgetTree.Button_Collapse.HandleOnButtonHovered)
    onButtonBaseUnhovered=(&self:WidgetTree.Button_Collapse.HandleOnButtonUnhovered)
    onSelectedChangedBase=(&self:WidgetTree.Button_Collapse.HandleOnSelectedChanged)
    visibility=Collapsed
   }
   slot {class=OverlaySlot, properties {horizontalAlignment=HAlign_Center, verticalAlignment=VAlign_Center}}
  }
  SaveSpinner_2 {
   widgetClassPath=&Image
   bIsVariable=true
   parent=Overlay_25
   widget=SaveSpinner_2
   properties {
    visibility=Collapsed
    brush {
     bIsDynamicallyLoaded=false
     drawAs=Image
     imageType=NoImage
     mirroring=NoMirror
     resourceName=None
     resourceObject=/FortUILibrary/UIKit/Throbber/MI_UIKit_Color_Throbber
     tiling=NoTile
     imageSize=<24 24>
     margin=<ltrb 0 0 0 0>
     outlineSettings {
      bUseBrushTransparency=false
      roundingType=HalfHeightRadius
      width=0
      color {colorUseRule=UseColor_Specified, specifiedColor=<rgba 0 0 0 0>}
      cornerRadii=<0 0 0 0>
     }
     tintColor {colorUseRule=UseColor_Specified, specifiedColor=<rgba 1 1 1 1>}
     uVRegion {bIsValid=false, max=<0 0>, min=<0 0>}
    }
   }
   slot {class=OverlaySlot, properties {verticalAlignment=VAlign_Center, padding=<ltrb 10 -24 0 0>}}
  }
 ]
}
