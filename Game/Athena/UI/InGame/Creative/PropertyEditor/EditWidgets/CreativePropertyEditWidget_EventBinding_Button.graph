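# Button widget for event bindings in the Creative property editor, with skewed background, primary content and an open-prompt row.
# Hover, selection, dialog, marker and reference-type changes feed Update Button Visuals, which restyles background and content and shows the prompt when highlighted.
# For device buttons, UpdateButtonStates maps reference type to default and hovered content states.
# A dirty flag defers swapping the arrow and forward/back input action until the button is highlighted.
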WidgetBlueprint /Game/Athena/UI/InGame/Creative/PropertyEditor/EditWidgets/CreativePropertyEditWidget_EventBinding_Button
graphs [
 EventGraph {nodes [
   n1 = Event OnIsDisplayingButtonOptionsSet!() -> (OutputDelegate, then, bIsDisplayingButtonOptions:bool) {is=title K2Node_Event}
   > Branch(Condition="Not Equal (Boolean)"(A=Get bIsShowingDialog() -> (bIsShowingDialog:bool), B=@n1.bIsDisplayingButtonOptions) -> (ReturnValue=~))
   > Set bIsShowingDialog(bIsShowingDialog:bool=@n1.bIsDisplayingButtonOptions) -> (Output_Get:bool)
   > Set bIsShowingDialogChanged(bIsShowingDialogChanged:bool=true) -> (Output_Get:bool)
   > Update Button Visuals()
   n2 = Event On Hovered()
   Update Button Visuals(execute=[@n2.then, @n4.then, @n5.then, @n6.then])
   n3 = Event OnTextSet!() -> (OutputDelegate, then, Text:"Text (by ref)") {is=title K2Node_Event}
   > UpdateText!(
    execute
    self:WBP Creative Property Editor Content Primary Object Reference=Get WBP_CreativePropertyEditor_Content_Primary() -> (WBP_CreativePropertyEditor_Content_Primary:WBP Creative Property Editor Content Primary Object Reference)
    Text:"Text (by ref)"=@n3.Text
   ) -> (then)
   n4 = Event On Unhovered()
   n5 = Event On Selected()
   n6 = Event On Deselected()
   n7 = Event OnEventReferenceTypeSet!() -> (OutputDelegate, then, EventReferenceType:EFortGameplayEventReferenceType Enum=None) {is=title K2Node_Event}
   > UpdateButtonStates!(execute, self, EventReferenceType:EFortGameplayEventReferenceType Enum=@n7.EventReferenceType) -> (then)
   > Update Button Visuals()
   Event PreConstruct()
   > UpdateButtonStates(EventReferenceType=None)
   > Update Button Visuals()
  ]}
 SetIsMarkerActive {nodes [
   n1 = SetIsMarkerActive!() -> (then, bInIsMarkerActive:bool=~) {is=title K2Node_FunctionEntry}
   > Branch(Condition="Not Equal (Boolean)"(A=@n1, B=Get bIsMarkerActive() -> (bIsMarkerActive:bool)) -> (ReturnValue=~))
   > Set bIsMarkerActive(bIsMarkerActive:bool=@n1) -> (Output_Get:bool)
   > Update Button Visuals()
  ]}
 Update Button Visuals {nodes [
   Update Button Visuals() {is=title K2Node_FunctionEntry}
   > Set bLocal_Selected(bLocal_Selected:bool=OR Boolean!(A=GetSelected!(self:Common Button Base Object Reference) -> (ReturnValue:bool), B=IsHovered(), C:bool=Get bIsShowingDialog()) -> (ReturnValue)) -> (Output_Get:bool)
   n1 = NOT Boolean(A=Get bLocal_Selected() -> (bLocal_Selected:bool))
   n2 = Get bIsMarkerActive()
   n3 = Get bIsDeviceButton() -> (bIsDeviceButton:bool)
   > Event Update Background Style!(
    execute
    self:WBP Creative Property Editor Background Skewed With Marker Object Reference=Get WBP_CreativePropertyEditor_Background_SkewWithMarker() -> (WBP_CreativePropertyEditor_Background_SkewWithMarker:WBP Creative Property Editor Background Skewed With Marker Object Reference)
    InType:E_UI_CreativePropertyEditor_Type Enum=Get Type() -> (Type:E_UI_CreativePropertyEditor_Type Enum=NewEnumerator4)
    InMarkerState:E_UI_CreativePropertyEditor_MarkerStates Enum=Select(
     Option 0:E_UI_CreativePropertyEditor_MarkerStates Enum=NewEnumerator0
     Option 1:E_UI_CreativePropertyEditor_MarkerStates Enum=NewEnumerator1
     Index=AND Boolean!(A=@n2, B=@n3, C:bool=@n1) -> (ReturnValue)
    ) -> (ReturnValue:E_UI_CreativePropertyEditor_MarkerStates Enum)
    InFocusState:E_UI_CreativePropertyEditor_FocusStates Enum=Select(
     Option 0:E_UI_CreativePropertyEditor_FocusStates Enum=NewEnumerator0
     Option 1:E_UI_CreativePropertyEditor_FocusStates Enum=NewEnumerator1
     Index=Get bLocal_Selected()
    ) -> (ReturnValue:E_UI_CreativePropertyEditor_FocusStates Enum)
   ) -> (then)
   > Event Update States!(
    execute
    self:WBP Creative Property Editor Content Primary Object Reference=Get WBP_CreativePropertyEditor_Content_Primary()
    State:E_UI_ChannelsBindingStates Enum=Select(
     Option 0:E_UI_ChannelsBindingStates Enum=Get DefaultState() -> (DefaultState:E_UI_ChannelsBindingStates Enum=NewEnumerator7)
     Option 1:E_UI_ChannelsBindingStates Enum=Get HoveredState() -> (HoveredState:E_UI_ChannelsBindingStates Enum=NewEnumerator7)
     Index=Get bLocal_Selected()
    ) -> (ReturnValue:E_UI_ChannelsBindingStates Enum)
   ) -> (then)
   n4 = Get bLocal_Selected()
   > SetVisibility(
    self=Get HBox_Open() -> (HBox_Open:Horizontal Box Object Reference)
    InVisibility=Select(Option 0:ESlateVisibility Enum=Collapsed, Option 1:ESlateVisibility Enum=SelfHitTestInvisible, Index=@n4) -> (ReturnValue:ESlateVisibility Enum)
   )
   > Branch(Condition=AND Boolean!(A=Get bIsShowingDialogChanged() -> (bIsShowingDialogChanged:bool), B=@n4) -> (ReturnValue))
   > Set bIsShowingDialogChanged(bIsShowingDialogChanged=false)
   > SetBrushResourceObject!(
    execute
    self:Image Object Reference=Get Image_Arrow() -> (Image_Arrow:Image Object Reference)
    ResourceObject:Object Reference=Select(
     Option 0:Object Reference=Get ArrowIcon_NotShowingDialog() -> (ArrowIcon_NotShowingDialog:Texture Object Reference)
     Option 1:Object Reference=Get ArrowIcon_ShowingDialog() -> (ArrowIcon_ShowingDialog:Texture Object Reference)
     Index=Get bIsShowingDialog()
    ) -> (ReturnValue:Object Reference)
   ) -> (then)
   > SetInputAction!(
    execute
    self:Common Action Widget Object Reference=Get Action_Open() -> (Action_Open:Common Action Widget Object Reference)
    InputActionRow:Data Table Row Handle Structure=Select(
     Option 0:Data Table Row Handle Structure=Get InputAction_NotShowingDialog() -> (InputAction_NotShowingDialog:Data Table Row Handle Structure)
     Option 1:Data Table Row Handle Structure=Get InputAction_ShowingDialog() -> (InputAction_ShowingDialog:Data Table Row Handle Structure)
     Index=Get bIsShowingDialog()
    ) -> (ReturnValue:Data Table Row Handle Structure)
   ) -> (then)
  ]}
 UpdateButtonStates {nodes [
   n1 = UpdateButtonStates!() -> (then, EventReferenceType:EFortGameplayEventReferenceType Enum=None) {is=title K2Node_FunctionEntry}
   > Branch(Condition=Get bIsDeviceButton())
   > Set DefaultState(DefaultState:E_UI_ChannelsBindingStates Enum=Select!(
     None:E_UI_ChannelsBindingStates Enum=NewEnumerator7
     EventDescriptor:E_UI_ChannelsBindingStates Enum=NewEnumerator2
     EventFunction:E_UI_ChannelsBindingStates Enum=NewEnumerator0
     Index:EFortGameplayEventReferenceType Enum=@n1
    ) -> (ReturnValue:E_UI_ChannelsBindingStates Enum)) -> (Output_Get:E_UI_ChannelsBindingStates Enum=NewEnumerator7)
   > Set HoveredState(HoveredState:E_UI_ChannelsBindingStates Enum=Select(None=NewEnumerator8, EventDescriptor=NewEnumerator3, EventFunction=NewEnumerator1, Index=@n1)) -> (Output_Get:E_UI_ChannelsBindingStates Enum=NewEnumerator7)
   > Set Type(Type:E_UI_CreativePropertyEditor_Type Enum=NewEnumerator1) -> (Output_Get:E_UI_CreativePropertyEditor_Type Enum=NewEnumerator4)
  ]}
]
variables [
 ArrowIcon_NotShowingDialog {type=Texture, default=/Game/UI/SocialPanel/Art/Textures/T_UI_Icon_TreeArrow_64}
 ArrowIcon_ShowingDialog {type=Texture, default=/Game/UI/SocialPanel/Art/Textures/T_UI_Icon_TreeArrow_64_reversed}
 bIsDeviceButton {type=boolean}
 bIsMarkerActive {type=boolean}
 bIsShowingDialog {type=boolean}
 bIsShowingDialogChanged {type=boolean}
 DefaultState {type=E_UI_ChannelsBindingStates, default=Default}
 HoveredState {type=E_UI_ChannelsBindingStates, default=DefaultHovered}
 InputAction_NotShowingDialog {type=DataTableRowHandle, default {dataTable=&InputActionDataTable, rowName=GenericForward}}
 InputAction_ShowingDialog {type=DataTableRowHandle, default {dataTable=&InputActionDataTable, rowName=GenericBack}}
 Type {type=E_UI_CreativePropertyEditor_Type, default=Default}
]
widgets {
 info {
  inheritedWidgetCount=2
  parentClass=/Script/FortniteUI.FortCreativePropertyEditWidgetEventBindingButton
  rootWidgetClass=&Overlay
  widgetCount=9
 }
 widgets [
  Overlay_1 {widgetClassPath=&Overlay, widget=Overlay_1}
  WBP_CreativePropertyEditor_Background_SkewWithMarker {
   widgetClassPath=/Game/Athena/UI/InGame/Creative/PropertyEditor/WBP_CreativePropertyEditor_Background_SkewedWithMarker
   bIsVariable=true
   parent=Overlay_1
   widget=WBP_CreativePropertyEditor_Background_SkewWithMarker
   properties {type=Default, displayLeftSkew=false, displayMarker=false, displayRightSkew=false, leftSkewEdgeWidth=14, rightSkewEdgeWidth=14}
   slot {class=OverlaySlot, properties {horizontalAlignment=HAlign_Fill, verticalAlignment=VAlign_Fill}}
  }
  HorizontalBox_0 {
   widgetClassPath=&HorizontalBox
   parent=Overlay_1
   widget=HorizontalBox_0
   slot {class=OverlaySlot, properties {horizontalAlignment=HAlign_Fill, verticalAlignment=VAlign_Center, padding=<ltrb 24 18 24 18>}}
  }
  WBP_CreativePropertyEditor_Content_Primary {
   widgetClassPath=/Game/Athena/UI/InGame/Creative/PropertyEditor/WBP_CreativePropertyEditor_Content_Primary
   bIsVariable=true
   parent=HorizontalBox_0
   widget=WBP_CreativePropertyEditor_Content_Primary
   properties {states=FunctionHovered}
   slot {class=HorizontalBoxSlot, properties {size {sizeRule=Fill, value=1}}}
  }
  HBox_Open {
   widgetClassPath=&HorizontalBox
   bIsVariable=true
   parent=HorizontalBox_0
   widget=HBox_Open
   slot {class=HorizontalBoxSlot, properties {padding=<ltrb 24 0 0 0>}}
  }
  Action_Open {
   widgetClassPath=&CommonActionWidget
   bIsVariable=true
   parent=HBox_Open
   widget=Action_Open
   properties {visibility=SelfHitTestInvisible, inputActions [{dataTable=&InputActionDataTable, rowName=GenericForward}]}
   slot {class=HorizontalBoxSlot, properties {verticalAlignment=VAlign_Center}}
  }
  Image_Arrow {
   widgetClassPath=&Image
   bIsVariable=true
   parent=HBox_Open
   widget=Image_Arrow
   properties {
    visibility=SelfHitTestInvisible
    brush {
     bIsDynamicallyLoaded=false
     drawAs=Image
     imageType=NoImage
     mirroring=NoMirror
     resourceName=None
     resourceObject=&T_UI_Icon_TreeArrow_64
     tiling=NoTile
     imageSize=<30 30>
     margin=<ltrb 0 0 0 0>
     outlineSettings {
      bUseBrushTransparency=false
      roundingType=HalfHeightRadius
      width=0
      color {colorUseRule=UseColor_Specified, specifiedColor=<rgba 0 0 0 0>}
      cornerRadii=<0 0 0 0>
     }
     tintColor {colorUseRule=UseColor_Specified, specifiedColor=<rgba 1 1 1 1>}
     uVRegion {bIsValid=false, max=<0 0>, min=<0 0>}
    }
    colorAndOpacity=<rgba 0.2502 0.2502 0.2502 1>
   }
   slot {class=HorizontalBoxSlot, properties {verticalAlignment=VAlign_Center}}
  }
  Text_Name {widgetClassPath=&CommonTextBlock, bInherited=true, widget=null}
  InputActionWidget {widgetClassPath=&CommonActionWidget, bInherited=true, widget=null}
 ]
}
