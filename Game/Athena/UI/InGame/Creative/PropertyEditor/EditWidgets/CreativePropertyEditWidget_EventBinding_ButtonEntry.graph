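# Entry row widget for event bindings in the Creative property editor, pairing a device select button with an option select button.
# The device button's marker is active while either the entry or its parent item is selected.
# The row sets itself SelfHitTestInvisible once both device and option display names are set.
# On Entry Released clears all four tracking flags without updating visibility or the marker.
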WidgetBlueprint /Game/Athena/UI/InGame/Creative/PropertyEditor/EditWidgets/CreativePropertyEditWidget_EventBinding_ButtonEntry
graphs [EventGraph {nodes [
   Event On Entry Released() {is=title K2Node_Event}
   > Set bIsParentItemSelected(bIsParentItemSelected:bool=false) -> (Output_Get:bool)
   > Set bIsEntrySelected(bIsEntrySelected:bool=false) -> (Output_Get:bool)
   > Set bIsDeviceDisplayNameSet(bIsDeviceDisplayNameSet:bool=false) -> (Output_Get:bool)
   > Set bIsOptionDisplayNameSet(bIsOptionDisplayNameSet:bool=false) -> (Output_Get:bool)
   n1 = Event OnEntrySelectedStateChanged!() -> (OutputDelegate, then, bIsSelected:bool) {is=title K2Node_Event}
   > n2 = Set bIsEntrySelected(bIsEntrySelected=@n1.bIsSelected)
   n3 = Cast To CreativePropertyEditWidget_EventBinding_Button!(
    execute=[@n5.then, @n2.then]
    Object=Get Button_SelectDevice() -> (Button_SelectDevice:Fort Creative Property Edit Widget Event Binding Button Object Reference)
   ) -> (then, CastFailed, AsCreative Property Edit Widget Event Binding Button:Creative Property Edit Widget Event Binding Button Object Reference) {is=title K2Node_DynamicCast}
   > SetIsMarkerActive!(
    execute
    self:Creative Property Edit Widget Event Binding Button Object Reference=@n3
    bInIsMarkerActive:bool=OR Boolean(A=Get bIsEntrySelected() -> (bIsEntrySelected:bool), B=Get bIsParentItemSelected() -> (bIsParentItemSelected:bool))
   ) -> (then)
   n4 = Event OnIsParentItemSelectedChanged!() -> (OutputDelegate, then, bIsParentItemSelected:bool) {is=title K2Node_Event}
   > n5 = Set bIsParentItemSelected(bIsParentItemSelected=@n4.bIsParentItemSelected)
   Event OnDeviceDisplayNameSet!() -> (OutputDelegate, then, DeviceDisplayName:"Text (by ref)") {is=title K2Node_Event}
   > n6 = Set bIsDeviceDisplayNameSet(bIsDeviceDisplayNameSet=true)
   Branch(
    execute=[@n6.then, @n7.then]
    Condition=AND Boolean(A=Get bIsDeviceDisplayNameSet() -> (bIsDeviceDisplayNameSet:bool), B=Get bIsOptionDisplayNameSet() -> (bIsOptionDisplayNameSet:bool))
   )
   > SetVisibility(InVisibility=SelfHitTestInvisible)
   Event OnOptionDisplayNameSet!() -> (OutputDelegate, then, OptionDisplayName:"Text (by ref)") {is=title K2Node_Event}
   > n7 = Set bIsOptionDisplayNameSet(bIsOptionDisplayNameSet=true)
  ]}]
variables [
 bIsDeviceDisplayNameSet {type=boolean}
 bIsEntrySelected {type=boolean}
 bIsOptionDisplayNameSet {type=boolean}
 bIsParentItemSelected {type=boolean}
]
widgets {
 info {
  inheritedWidgetCount=2
  parentClass=/Script/FortniteUI.FortCreativePropertyEditWidgetEventBindingEntry
  rootWidgetClass=&HorizontalBox
  widgetCount=4
 }
 widgets [
  HBox_EventButtons {widgetClassPath=&HorizontalBox, widget=HBox_EventButtons}
  Button_SelectDevice {
   widgetClassPath=&/CreativePropertyEditWidget_EventBinding_Button
   bInherited=true
   parent=HBox_EventButtons
   widget=Button_SelectDevice
   properties {bIsDeviceButton=true}
   slot {class=HorizontalBoxSlot, properties {padding=<ltrb 16 0 0 0>, size {sizeRule=Fill, value=1}}}
  }
  Spacer_134 {
   widgetClassPath=&Spacer
   parent=HBox_EventButtons
   widget=Spacer_134
   properties {visibility=Visible, size=<6 1>}
   slot {class=HorizontalBoxSlot}
  }
  Button_SelectOption {
   widgetClassPath=&CreativePropertyEditWidget_EventBinding_Button
   bInherited=true
   parent=HBox_EventButtons
   widget=Button_SelectOption
   slot {class=HorizontalBoxSlot, properties {padding=<ltrb 0 0 16 0>, size {sizeRule=Fill, value=1}}}
  }
 ]
}
