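# Base row widget for event binding properties in the Creative property editor, built on FortCreativePropertyEditWidgetEventBinding.
# Native selection, bindings, naming, modified and count events update cached flags or push state into child content and skewed background widgets.
# Content state follows selection and event-function type, while background reflects parent status, focus and sub-item selection.
# The Add button appears only with available devices and either selection or touch input.
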WidgetBlueprint /Game/Athena/UI/InGame/Creative/PropertyEditor/EditWidgets/CreativePropertyEditWidget_EventBinding_Base
graphs [
 CheckIfEntryIsParent {nodes [
   CheckIfEntryIsParent() {is=title K2Node_FunctionEntry}
   > n1 = Branch(Condition=OR Boolean(
     A=Get bHasAtLeastOneEvent() -> (bHasAtLeastOneEvent:bool)
     B=IsShowingBindings!(self:Fort Creative Property Edit Widget Event Binding Object Reference) -> (ReturnValue:bool)
    ))
   > Set bIsParent(bIsParent:bool=true) -> (Output_Get:bool)
   Set bIsParent(execute=@n1.else, bIsParent=false)
  ]}
 CheckIfEventExists {nodes [
   n1 = CheckIfEventExists!() -> (then, ExistingEventNumbers:int) {is=title K2Node_FunctionEntry}
   > n2 = Branch(Condition="integer > integer"(A=@n1))
   > Set bHasAtLeastOneEvent(bHasAtLeastOneEvent:bool=true) -> (Output_Get:bool)
   Set bHasAtLeastOneEvent(execute=@n2.else, bHasAtLeastOneEvent=false)
  ]}
 EventGraph {nodes [
   UpdateVisuals() {is=title K2Node_CustomEvent}
   > UpdateAddBindingVisibility()
   > CheckIfEntryIsParent() {is=title K2Node_CallFunction}
   > UpdateBackground()
   n1 = Get bIsEventFunction() -> (bIsEventFunction:bool)
   > Event Update States!(
    execute
    self:WBP Creative Property Editor Content Primary Object Reference=Get WBP_CreativePropertyEditor_Content_Primary() -> (WBP_CreativePropertyEditor_Content_Primary:WBP Creative Property Editor Content Primary Object Reference)
    State:E_UI_ChannelsBindingStates Enum=Select(
     Option 0:E_UI_ChannelsBindingStates Enum=Select(Option 0=NewEnumerator2, Option 1=NewEnumerator0, Index=@n1)
     Option 1:E_UI_ChannelsBindingStates Enum=Select(Option 0=NewEnumerator3, Option 1=NewEnumerator1, Index=@n1)
     Index=Get bIsSelected() -> (bIsSelected:bool)
    ) -> (ReturnValue:E_UI_ChannelsBindingStates Enum)
   ) -> (then)
   Event OnShowBindings() {is=title K2Node_Event}
   > UpdateVisuals() {is=title K2Node_CallFunction}
   Event OnHideBindings() {is=title K2Node_Event}
   > UpdateVisuals()
   n2 = Event OnEditWidgetSelectedChanged!() -> (OutputDelegate, then, bSelected:bool) {is=title K2Node_Event}
   > Set bIsSelected(bIsSelected:bool=@n2.bSelected) -> (Output_Get:bool)
   > UpdateVisuals()
   n3 = Event OnPropertyNameSet!() -> (OutputDelegate, then, PropertyName:"Text (by ref)") {is=title K2Node_Event}
   > UpdateVisuals()
   > UpdateText!(
    execute
    self:WBP Creative Property Editor Content Primary Object Reference=Get WBP_CreativePropertyEditor_Content_Primary()
    Text:"Text (by ref)"=@n3.PropertyName
   ) -> (then)
   n4 = Event BP_OnModified!() -> (OutputDelegate, then, bModified:bool) {is=title K2Node_Event}
   > Event Update IsModified!(
    execute
    self:WBP Creative Property Editor Content Primary Object Reference=Get WBP_CreativePropertyEditor_Content_Primary()
    bIsModified:bool=@n4.bModified
   ) -> (then)
   n5 = Event OnIsAnySubItemSelectedChanged!() -> (OutputDelegate, then, bIsAnySubItemSelected:bool) {is=title K2Node_Event}
   > Set bIsAnySubItemSelected(bIsAnySubItemSelected:bool=@n5.bIsAnySubItemSelected) -> (Output_Get:bool)
   > UpdateBackground()
   Event Construct()
   > UpdateVisuals()
   n6 = Event OnNumEventReferencesChanged!() -> (OutputDelegate, then, NumEventReferences:int=0) {is=title K2Node_Event}
   > CheckIfEventExists!(execute, self, ExistingEventNumbers:int=@n6.NumEventReferences) -> (then) {is=title K2Node_CallFunction}
   n7 = Event OnNumAvailableDevicesChanged!() -> (OutputDelegate, then, NumAvailableDevices:int=0) {is=title K2Node_Event}
   > Set bHasAtLeastOneAvailableDevice(bHasAtLeastOneAvailableDevice:bool="integer > integer"(A=@n7.NumAvailableDevices)) -> (Output_Get:bool)
   > UpdateAddBindingVisibility()
  ]}
 UpdateAddBindingVisibility {nodes [
   UpdateAddBindingVisibility() {is=title K2Node_FunctionEntry}
   n1 = OR Boolean(A=Get bIsSelected(), B=IsUsingTouch(PlayerController=GetOwningPlayer()))
   n2 = Get bHasAtLeastOneAvailableDevice() -> (bHasAtLeastOneAvailableDevice:bool)
   > SetVisibility(
    self=Get Button_AddBinding() -> (Button_AddBinding:Common Button Base Object Reference)
    InVisibility=Select(Option 0:ESlateVisibility Enum=Collapsed, Option 1:ESlateVisibility Enum=SelfHitTestInvisible, Index=AND Boolean(A=@n1, B=@n2)) -> (ReturnValue:ESlateVisibility Enum)
   )
  ]}
 UpdateBackground {nodes [
   UpdateBackground() {is=title K2Node_FunctionEntry}
   n1 = IsShowingBindings()
   n2 = NOT Boolean(A=Get bIsSelected())
   n3 = Get bIsAnySubItemSelected() -> (bIsAnySubItemSelected:bool)
   n4 = Get bIsSelected()
   > Event Update Background Style!(
    execute
    self:WBP Creative Property Editor Background Skewed With Marker Object Reference=Get WBP_CreativePropertyEditor_Background_SkewWithMarker() -> (WBP_CreativePropertyEditor_Background_SkewWithMarker:WBP Creative Property Editor Background Skewed With Marker Object Reference)
    InType:E_UI_CreativePropertyEditor_Type Enum=Select(
     Option 0:E_UI_CreativePropertyEditor_Type Enum=Select(Option 0=NewEnumerator4, Option 1=NewEnumerator3, Index=@n4)
     Option 1:E_UI_CreativePropertyEditor_Type Enum=NewEnumerator2
     Index=Get bIsParent() -> (bIsParent:bool)
    ) -> (ReturnValue:E_UI_CreativePropertyEditor_Type Enum)
    InMarkerState:E_UI_CreativePropertyEditor_MarkerStates Enum=Select(
     Option 0:E_UI_CreativePropertyEditor_MarkerStates Enum=NewEnumerator0
     Option 1:E_UI_CreativePropertyEditor_MarkerStates Enum=NewEnumerator1
     Index=AND Boolean!(A=@n3, B=@n1, C:bool=@n2) -> (ReturnValue)
    ) -> (ReturnValue:E_UI_CreativePropertyEditor_MarkerStates Enum)
    InFocusState:E_UI_CreativePropertyEditor_FocusStates Enum=Select(
     Option 0:E_UI_CreativePropertyEditor_FocusStates Enum=NewEnumerator0
     Option 1:E_UI_CreativePropertyEditor_FocusStates Enum=NewEnumerator1
     Index=Get bIsSelected()
    ) -> (ReturnValue:E_UI_CreativePropertyEditor_FocusStates Enum)
   ) -> (then)
  ]}
]
variables [
 bHasAtLeastOneAvailableDevice {type=boolean}
 bHasAtLeastOneEvent {type=boolean}
 bIsAnySubItemSelected {type=boolean}
 bIsEventFunction {type=boolean}
 bIsParent {type=boolean}
 bIsSelected {type=boolean}
]
widgets {
 info {
  inheritedWidgetCount=3
  parentClass=/Script/FortniteUI.FortCreativePropertyEditWidgetEventBinding
  rootWidgetClass=&SizeBox
  widgetCount=8
 }
 widgets [
  SizeBox_2 {widgetClassPath=&SizeBox, widget=SizeBox_2, properties {bOverride_HeightOverride=true, heightOverride=82}}
  Overlay_0 {widgetClassPath=&Overlay, parent=SizeBox_2, widget=Overlay_0, slot {class=SizeBoxSlot, properties {padding=<ltrb 16 0 16 0>}}}
  WBP_CreativePropertyEditor_Background_SkewWithMarker {
   widgetClassPath=/Game/Athena/UI/InGame/Creative/PropertyEditor/WBP_CreativePropertyEditor_Background_SkewedWithMarker
   bIsVariable=true
   parent=Overlay_0
   widget=WBP_CreativePropertyEditor_Background_SkewWithMarker
   properties {type=Parent, displayLeftSkew=false, displayMarker=false, displayRightSkew=false, focusState=Hovered}
   slot {class=OverlaySlot, properties {horizontalAlignment=HAlign_Fill, verticalAlignment=VAlign_Fill}}
  }
  HorizontalBox_0 {
   widgetClassPath=&HorizontalBox
   parent=Overlay_0
   widget=HorizontalBox_0
   slot {class=OverlaySlot, properties {horizontalAlignment=HAlign_Fill, verticalAlignment=VAlign_Center, padding=<ltrb 12 0 12 0>}}
  }
  WBP_CreativePropertyEditor_Content_Primary {
   widgetClassPath=/Game/Athena/UI/InGame/Creative/PropertyEditor/WBP_CreativePropertyEditor_Content_Primary
   bIsVariable=true
   parent=HorizontalBox_0
   widget=WBP_CreativePropertyEditor_Content_Primary
   properties {states=FunctionHovered}
   slot {class=HorizontalBoxSlot, properties {padding=<ltrb 6 0 0 0>, size {sizeRule=Fill, value=1}}}
  }
  Button_AddBinding {
   widgetClassPath=&IconTextButton
   bInherited=true
   parent=HorizontalBox_0
   widget=Button_AddBinding
   properties {
    bIsFocusable=false
    buttonText=Add
    onButtonBaseClicked=(&self:WidgetTree.Button_AddBinding.HandleOnButtonClicked)
    onButtonBaseDoubleClicked=(&self:WidgetTree.Button_AddBinding.HandleOnButtonDoubleClicked)
    onButtonBaseHovered=(&self:WidgetTree.Button_AddBinding.HandleOnButtonHovered)
    onButtonBaseUnhovered=(&self:WidgetTree.Button_AddBinding.HandleOnButtonUnhovered)
    onSelectedChangedBase=(&self:WidgetTree.Button_AddBinding.HandleOnSelectedChanged)
    style=/FortUILibrary/Foundation/Buttons/ButtonStyles/Athena/Athena_ButtonStyle_AngledBlueMenuButton.Athena_ButtonStyle_AngledBlueMenuButton_C
    triggeringInputAction {dataTable=&InputActionDataTable, rowName=GenericForward}
   }
   slot {class=HorizontalBoxSlot, properties {padding=<ltrb 24 10 0 10>}}
  }
  Text_PropertyName {widgetClassPath=&CommonTextBlock, bInherited=true, widget=null}
  Button_Collapse {widgetClassPath=&CommonButtonLegacy, bInherited=true, widget=null}
 ]
}
