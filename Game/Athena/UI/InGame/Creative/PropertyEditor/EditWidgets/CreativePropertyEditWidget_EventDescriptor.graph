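# Widget blueprint for a Creative property editor edit widget named for event descriptor properties.
# It is an empty child of CreativePropertyEditWidget_EventBinding_Base, adding no widgets, variables or graph logic of its own.
# All layout and behavior come from that parent event-binding widget.
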WidgetBlueprint /Game/Athena/UI/InGame/Creative/PropertyEditor/EditWidgets/CreativePropertyEditWidget_EventDescriptor
graphs [EventGraph {links [], nodes []}]
variables []
widgets {
 info {
  inheritedWidgetCount=0
  parentClass=&/CreativePropertyEditWidget_EventBinding_Base.CreativePropertyEditWidget_EventBinding_Base_C
  rootWidgetClass=null
  widgetCount=0
 }
 widgets []
}
